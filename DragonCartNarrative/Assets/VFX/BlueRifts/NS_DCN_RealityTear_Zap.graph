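# Niagara effect for a reality tear zap: pulsing electric ribbons and tethers around a tree position, timed in staggered bursts.
# ElectricityDown arcs descend to the tree point, while ElectricityUp and TetherRibbons curve from it toward a fixed local offset.
# Partitioned bursts form ribbons shaped by index curves, with a traveling pulse boosting width and color on the electricity.
# TetherMotes spawn from and follow the tether ribbons; ElectricArcs and Flare emitters are disabled.
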NiagaraSystem /DragonCartNarrative/Assets/VFX/BlueRifts/NS_DCN_RealityTear_Zap
dependencies {
 usedDataInterfaces [
  &NiagaraDataInterfaceCurve
  &NiagaraDataInterfaceVectorCurve
  &NiagaraDataInterfaceParticleRead
  &NiagaraDataInterfaceColorCurve
  &NiagaraDataInterfaceArrayFloat
  &NiagaraDataInterfaceVector2DCurve
 ]
 usedDynamicInputs [
  &SimulationPosition
  &Vector2DFromFloat
  &TransformPosition
  /Niagara/DynamicInputs/Vectors/CrossProduct
 ]
 usedModules [
  &self:SystemSpawnScript.SystemScriptSource.SystemScriptGraph.NiagaraNodeAssignment_2.SetVariables_#1
  &self:SystemSpawnScript.SystemScriptSource.SystemScriptGraph.NiagaraNodeAssignment_3.SetVariables_#2
  &self:SystemSpawnScript.SystemScriptSource.SystemScriptGraph.NiagaraNodeAssignment_0.SetVariables_#3
  &SystemState
 ]
}
emitters {
 ElectricArcs {
  bEnabled=false
  emitterUpdateScript [
   &EmitterState()
   &SpawnRate(
    Spawn Probability=0.4
    SpawnRate=&FloatFromCurve(
     FloatCurve=DataInterface() {propertyValues {Curve {
        defaultValue=1
        keys [
         {arriveTangent=1.7638, interpMode=RCIM_Cubic, time=0, value=0}
         {arriveTangentWeight=0.1094, interpMode=RCIM_Cubic, leaveTangentWeight=0.3287, tangentMode=RCTM_User, time=0.2129, value=1}
         {arriveTangentWeight=0.3239, interpMode=RCIM_Cubic, leaveTangentWeight=0.0746, tangentMode=RCTM_User, time=1, value=1}
         {arriveTangent=-7.2244, interpMode=RCIM_Cubic, tangentMode=RCTM_Break, time=1.1, value=0.1}
        ]
       }}}
     CurveIndex=$Emitter.Age
     Scale Curve=80
    )
    Use Spawn Probability=true
   )
   &SpawnBurst_Instantaneous(Spawn Count=4, Spawn Group=1, Spawn Time=0.3, Use Spawn Probability=false, UseLoopCountLimit=false) {module=SpawnBurst_Instantaneous005}
   &SpawnBurst_Instantaneous(Spawn Count=3, Spawn Group=1, Spawn Time=0.5, Use Spawn Probability=false, UseLoopCountLimit=false)
   &SpawnBurst_Instantaneous(Spawn Count=2, Spawn Time=0.6, Use Spawn Probability=false, UseLoopCountLimit=false) {module=SpawnBurst_Instantaneous001}
   &SpawnBurst_Instantaneous(Spawn Count=3, Spawn Time=1.25, Use Spawn Probability=false, UseLoopCountLimit=false) {module=SpawnBurst_Instantaneous002}
   &SpawnBurst_Instantaneous(Spawn Count=4, Spawn Time=1.5, Use Spawn Probability=false, UseLoopCountLimit=false) {module=SpawnBurst_Instantaneous003}
   &SpawnBurst_Instantaneous(Spawn Count=6, Spawn Time=1.9, Use Spawn Probability=false, UseLoopCountLimit=false) {module=SpawnBurst_Instantaneous004}
  ]
  particleSpawnScript [
   /SetVariables(Particles.TreeSpawn:NiagaraBool=&SetBoolByFloatComparison(A=$Emitter.Age, B=1.15, Comparison Type=::NewEnumerator6)) {isSetParameters=true}
   &InitializeParticle(
    Color=&LerpLinearColors(
     StartColor=<rgba 0.1626 1.3472 4 1>
     EndColor=<rgba 4 0.1293 1.2944 1>
     LerpFactor=&FloatFromCurve(
      FloatCurve=DataInterface() {propertyValues {Curve {defaultValue=1, keys [<at 0.6 0 RCIM_Cubic>, {tangentMode=RCTM_User, time=0.8, value=1}]}}}
      CurveIndex=$Emitter.NormalizedLoopAge
     )
    )
    Lifetime=&UniformRangedFloat(Minimum=0.09, Maximum=0.125)
    Mass Mode=::NewEnumerator1
    Position Offset=&MakeCustomVectorFromBool(
     Bool=$Particles.TreeSpawn
     True Vector=&AddVectorToPosition(
      Position=$System.LocalTreePosition
      Vector=&MakeVector(
       X=0
       Y=0
       Z=&FloatFromCurve(FloatCurve=DataInterface() {propertyValues {Curve {keys [<at 0 1>,<at 1 0>]}}}, CurveIndex=$Emitter.Age, Scale Curve=1000)
      )
     )
     False Vector=<1 0 0>
    )
    Sprite Rotation Mode=::NewEnumerator2
    Sprite Size=&Multiply_Vector2D_ByFloat(
     Float=&FloatFromCurve(
      FloatCurve=DataInterface() {propertyValues {Curve {defaultValue=1, keys [<at 0 0.4 RCIM_Cubic>, {tangentMode=RCTM_User, time=1, value=1}]}}}
      CurveIndex=$Emitter.NormalizedLoopAge
     )
     Vector2D=&UniformRangedVector2D(Minimum=<3000 3000>, Maximum=<5000 5000>)
    )
    Sprite Size Mode=::NewEnumerator4
    UsePositionOffset=true
    Write Material Random=true
   ) {module=InitializeParticle001}
   &ShapeLocation(
    MaskSpawnGroup=false
    Override Seed=false
    Sphere Radius=&FloatFromCurve(
     FloatCurve=DataInterface() {propertyValues {Curve {
        defaultValue=1
        keys [
         <at 0.0067 0.5 RCIM_Cubic>
         {arriveTangentWeight=0.1695, interpMode=RCIM_Cubic, leaveTangentWeight=0.5155, tangentMode=RCTM_User, time=0.5, value=0.1}
         {arriveTangent=1.5678, tangentMode=RCTM_User, time=2, value=1}
        ]
       }}}
     CurveIndex=$Emitter.Age
     Scale Curve=1500
    )
    Sphere Surface Distribution=1
    UseOwnerScale=true
   )
   &CameraOffset(Camera Offset Amount=&RandomRangeFloat(Minimum=-50, Maximum=100))
  ]
  particleUpdateScript [
   &ParticleState()
   &DynamicMaterialParameters(
    Index 0 Param 1=$Particles.MaterialRandom
    Index 0 Param 2=&Lerp_Float(A=0.75, B=1.2, Alpha=$Particles.MaterialRandom)
    Index 0 Param 3=&FloatFromCurve(FloatCurve=DataInterface() {propertyValues {Curve {keys [<at 0.5 0 RCIM_Cubic>, {interpMode=RCIM_Cubic, leaveTangent=3.7979, time=1, value=1}]}}})
    Index 0 Param 4=1
    Index 1 Param 2=&FloatFromCurve(FloatCurve=DataInterface() {propertyValues {Curve {keys [<at 0.25 1 RCIM_Cubic>,<at 0.5 0>]}}}, CurveIndex=$System.Age, Scale Curve=0.5)
    Index 1 Param 3=&FloatFromCurve(
     FloatCurve=DataInterface() {propertyValues {Curve {keys [
         {arriveTangent=4.106, interpMode=RCIM_Cubic, leaveTangent=4.106, tangentMode=RCTM_User, time=0.05, value=0}
         {tangentMode=RCTM_User, time=0.35, value=1}
        ]}}}
     CurveIndex=$System.Age
    )
    Index 1 Param 4=100
    Module.Index0Float0Write=true
    Module.Index0Float1Write=true
    Module.Index0Float2Write=true
    Module.Index0Float3Write=true
    Module.Index1Float0Write=true
    Module.Index1Float1Write=true
    Module.Index1Float2Write=true
    Module.Index1Float3Write=true
    Write Parameter Index 1=::NewEnumerator1
   ) {module=DynamicMaterialParameters001}
   &ScaleSpriteSize(
    Uniform Curve Scale=&MakeCustomFloatFromBool(Bool=$Particles.TreeSpawn, True Float=0.25, False Float=1)
    Uniform Curve Sprite Scale=DataInterface() {propertyValues {Curve {keys [
        {arriveTangent=0.5666, interpMode=RCIM_Cubic, leaveTangent=0.5666, tangentMode=RCTM_User, time=0, value=0.75}
        {tangentMode=RCTM_User, time=1, value=1}
       ]}}}
   )
   &ScaleColor(
    Scale RGB=&MakeVectorFromLinearColorRGB(Color=&ColorFromCurve(ColorCurve=DataInterface() {propertyValues {
        AlphaCurve {keys []}
        BlueCurve {keys [<at 0.0806 4 RCIM_Cubic>,<at 0.4309 1.5>]}
        GreenCurve {keys [<at 0.0806 4 RCIM_Cubic>,<at 0.4309 0.7283>]}
        RedCurve {keys [<at 0.0806 4 RCIM_Cubic>,<at 0.4309 0.6641>]}
       }}))
    ScaleA=false
    ScaleRGB=true
   ) {module=ScaleColor001}
  ]
  rendererClasses [&NiagaraSpriteRendererProperties]
  renderers [{rendererClass=&NiagaraSpriteRendererProperties}]
 }
 ElectricityDown {
  emitterUpdateScript [
   &EmitterState()
   &SpawnBurst_Instantaneous(Spawn Count=20, Use Spawn Probability=false, UseLoopCountLimit=false)
   &SpawnBurst_Instantaneous(Spawn Count=60, Spawn Time=0.25, Use Spawn Probability=false, UseLoopCountLimit=false) {module=SpawnBurst_Instantaneous002}
   &SpawnBurst_Instantaneous(Spawn Count=20, Spawn Time=0.4, Use Spawn Probability=false, UseLoopCountLimit=false) {module=SpawnBurst_Instantaneous001}
  ]
  particleSpawnScript [
   &PartitionParticles(
    Normalized Index In Partition Includes One=true
    Particles Per Partition=20
    Partition Count Mode:ENiagara_InfiniteLoopDuration=::NewEnumerator1
    Partition Mode:ENiagara_PartitionMode_v2=::NewEnumerator1
    Write Partition Index to Ribbon Link Order:ENiagara_AttributeSamplingApplyOutput=::NewEnumerator0
    Write Partition to Ribbon ID=::NewEnumerator0
   )
   &InitializeParticle(
    Color=<rgba 0.0469 1.2144 1.5 1>
    Lifetime=0.6
    Ribbon Width=&FloatFromCurve(
     FloatCurve=DataInterface() {propertyValues {Curve {keys [<at 0 0.2>,<at 0.1 1>,<at 1 1.25>]}}}
     CurveIndex=$Particles.PartitionParticles.Normalized Index In Partition
     Scale Curve=30
    )
    Ribbon Width Mode=::NewEnumerator1
    Sprite Size Mode=::NewEnumerator3
    Uniform Sprite Size=50
    UsePositionOffset=false
    Write Material Random=true
   )
   /SetVariables(Particles.StartOffset:Vector3f=&MakeVector(
     X=-500
     Y=0
     Z=&FloatFromCurve(
      FloatCurve=DataInterface() {propertyValues {Curve {keys [{arriveTangent=-1.0266, interpMode=RCIM_Cubic, leaveTangent=-1.0266, tangentMode=RCTM_User, time=0, value=1}, <at 1 0.6>]}}}
      CurveIndex=$Emitter.Age
      Scale Curve=2200
     )
    )) {isSetParameters=true}
  ]
  particleUpdateScript [
   &ParticleState()
   /SetVariables(Particles.Pulse:NiagaraFloat=&FloatFromCurve(
     FloatCurve=DataInterface() {propertyValues {Curve {
        defaultValue=1
        keys [
         {arriveTangent=10.3786, interpMode=RCIM_Cubic, leaveTangent=10.3786, tangentMode=RCTM_User, time=0, value=0}
         {interpMode=RCIM_Cubic, tangentMode=RCTM_User, time=0.15, value=1}
         {arriveTangentWeight=0.1286, interpMode=RCIM_Cubic, leaveTangentWeight=0.2527, tangentMode=RCTM_User, time=0.35, value=0}
        ]
       }}}
     CurveIndex=&Add_Float(
      A=&Multiply_Float(A=$Particles.PartitionParticles.Normalized Index In Partition, B=-0.5)
      B=$Particles.NormalizedAge
     )
     Scale Curve=&FloatFromCurve(
      FloatCurve=DataInterface() {propertyValues {Curve {
         defaultValue=1
         keys [
          {arriveTangent=8.6346, interpMode=RCIM_Cubic, leaveTangent=8.6346, tangentMode=RCTM_User, time=0, value=0}
          {interpMode=RCIM_Cubic, tangentMode=RCTM_User, time=0.25, value=1}
          {arriveTangentWeight=0.3291, interpMode=RCIM_Cubic, leaveTangentWeight=0.1194, tangentMode=RCTM_User, time=0.7604, value=1}
          {arriveTangent=-3.1649, tangentMode=RCTM_User, time=1, value=0.5}
         ]
        }}}
      CurveIndex=$Particles.PartitionParticles.Normalized Index In Partition
     )
    )) {isSetParameters=true, module=SetVariables_#4}
   /SetVariables(Particles.Position=&AddVectorToPosition(
     Position=&Lerp_Vector(
      A=&AddVectorToPosition(Position=$System.LocalTreePosition, Vector=$Particles.StartOffset)
      B=Linked() {linkedVariable {name=System.LocalTreePosition, type=NiagaraPosition}}
      Alpha=$Particles.PartitionParticles.Normalized Index In Partition
     )
     Vector=&RotateVectorByQuaternion(
      Vector To Rotate=&Multiply_VectorByFloat(
       Vector=$System.SideDirection
       Float=&Add_Float(
        A=&FloatFromCurve(
         FloatCurve=DataInterface() {propertyValues {Curve {keys [
             {arriveTangent=6.2911, interpMode=RCIM_Cubic, leaveTangent=6.2911, tangentMode=RCTM_User, time=0, value=0.2}
             {arriveTangentWeight=0.0174, interpMode=RCIM_Cubic, leaveTangentWeight=0.0242, tangentMode=RCTM_User, time=0.3, value=1}
             {arriveTangent=-1.1429, interpMode=RCIM_Cubic, leaveTangent=-1.1429, time=0.6405, value=0.5201}
             {arriveTangent=-1.9281, interpMode=RCIM_Cubic, leaveTangent=-1.9281, tangentMode=RCTM_User, time=1, value=0.2}
            ]}}}
         CurveIndex=$Particles.PartitionParticles.Normalized Index In Partition
         Scale Curve=800
        )
        B=&Multiply_Float(A=$Particles.Pulse, B=100)
       )
      )
      Quaternion=&MakeQuaternion(
       Quaternion From=0
       Axis=<0 0 1>
       Angle=&Add_Float(
        A=&Multiply_FloatByInt(Float=0.3, Integer=$Particles.PartitionParticles.Partition)
        B=&FloatFromCurve(
         FloatCurve=DataInterface() {propertyValues {Curve {
            defaultValue=1
            keys [{arriveTangent=2.6402, interpMode=RCIM_Cubic, time=0, value=0}, {arriveTangent=1.4565, tangentMode=RCTM_User, time=1, value=1}]
           }}}
         CurveIndex=$Particles.PartitionParticles.Normalized Index In Partition
         Scale Curve=&SelectFloatFromArray(
          Direct Array Index=$Particles.PartitionParticles.Partition
          Direct Array Mode=::NewEnumerator1
          Float Selection Array=DataInterface() {propertyValues {FloatData [-0.75,0.75,0.5]}}
         )
        )
       )
       From=?
       To=?
       Angle Type=::NewEnumerator1
       Normalize Vectors=false
      )
     )
    )) {isSetParameters=true, module=SetVariables_#5}
   &self:ScratchModule_01(
    EndPos:NiagaraPosition=$System.LocalTreePosition
    StartPos:NiagaraPosition=&AddVectorToPosition(Position=$System.LocalTreePosition, Vector=<0 0 2000>)
   ) {module=ScratchModule_01}
   &DynamicMaterialParameters(
    Index 0 Param 1=$Particles.ScratchModule_01.X
    Index 0 Param 2=$Particles.ScratchModule_01.Y
    Index 0 Param 3=$Particles.ScratchModule_01.Z
    Index 0 Param 4=&Add_Float(A=$Particles.ScratchModule_01.W, B=1)
    Index 1 Param 1=$Particles.MaterialRandom
    Index 1 Param 2=&FloatFromCurve(FloatCurve=DataInterface() {propertyValues {Curve {
        defaultValue=1
        keys [
         {arriveTangent=1.5024, interpMode=RCIM_Cubic, leaveTangent=1.5024, tangentMode=RCTM_User, time=0, value=0.1}
         {arriveTangent=0.8224, interpMode=RCIM_Cubic, leaveTangent=0.8224, tangentMode=RCTM_User, time=1, value=1}
        ]
       }}})
    Index 1 Param 3=1.5
    Index 1 Param 4=&Lerp_Float(
     A=1
     B=0
     Alpha=/Niagara/DynamicInputs/Math/Min_Float(
      A:NiagaraFloat=&FloatFromCurve(
       FloatCurve=DataInterface() {propertyValues {Curve {
          defaultValue=1
          keys [
           {arriveTangent=17.5823, interpMode=RCIM_Cubic, leaveTangent=17.5823, leaveTangentWeight=0.05, tangentMode=RCTM_User, time=0, value=0}
           <at 0.1 1 RCIM_Cubic>
          ]
         }}}
       CurveIndex=$Particles.PartitionParticles.Normalized Index In Partition
      )
      B:NiagaraFloat=&FloatFromCurve(FloatCurve=DataInterface() {propertyValues {Curve {
          defaultValue=1
          keys [
           {interpMode=RCIM_Cubic, tangentMode=RCTM_User, time=0.7, value=1}
           {
            arriveTangent=-2.9651
            arriveTangentWeight=0.1286
            interpMode=RCIM_Cubic
            leaveTangent=-2.9651
            leaveTangentWeight=0.2527
            tangentMode=RCTM_User
            time=1
            value=0
           }
          ]
         }}})
     )
    )
    Module.Index0Float0Write=true
    Module.Index0Float1Write=true
    Module.Index0Float2Write=true
    Module.Index0Float3Write=true
    Module.Index1Float0Write=true
    Module.Index1Float1Write=true
    Module.Index1Float2Write=true
    Module.Index1Float3Write=true
    Write Parameter Index 1=::NewEnumerator1
   )
   &ScaleColor(Scale RGB=&Lerp_Vector(A=<1 1 1>, B=<5 0.5 1>, Alpha=$Particles.Pulse), ScaleA=false, ScaleRGB=true)
   &ScaleRibbonWidth(Ribbon Width Scale=&Lerp_Float(A=1, B=4, Alpha=$Particles.Pulse))
   /SetVariables(
    Particles.Airglow_Color:LinearColor=&MultiplyLinearColor(Linear Color=$Particles.Color, Scale Factor=<1.2 0.7 1 0.25>)
    Particles.Airglow_Width:NiagaraFloat=&Multiply_Float(A=$Particles.RibbonWidth, B=3)
   ) {isSetParameters=true, module=SetVariables_#6}
  ]
  rendererClasses [&NiagaraSpriteRendererProperties, &NiagaraRibbonRendererProperties]
  renderers [
   {rendererClass=&NiagaraSpriteRendererProperties}
   {rendererClass=&NiagaraRibbonRendererProperties, rendererIndex=1}
   {rendererClass=&NiagaraRibbonRendererProperties, rendererIndex=2}
  ]
 }
 ElectricityUp {
  emitterUpdateScript [
   &EmitterState()
   &SpawnBurst_Instantaneous(Spawn Count=20, Spawn Time=0.75, Use Spawn Probability=false, UseLoopCountLimit=false)
   &SpawnBurst_Instantaneous(Spawn Count=20, Spawn Time=1.1, Use Spawn Probability=false, UseLoopCountLimit=false) {module=SpawnBurst_Instantaneous002}
   &SpawnBurst_Instantaneous(Spawn Count=20, Spawn Group=1, Spawn Time=1.25, Use Spawn Probability=false, UseLoopCountLimit=false) {module=SpawnBurst_Instantaneous001}
  ]
  particleSpawnScript [
   &PartitionParticles(
    Normalized Index In Partition Includes One=true
    Particles Per Partition=20
    Partition Count Mode:ENiagara_InfiniteLoopDuration=::NewEnumerator1
    Partition Mode:ENiagara_PartitionMode_v2=::NewEnumerator1
    Write Partition Index to Ribbon Link Order:ENiagara_AttributeSamplingApplyOutput=::NewEnumerator0
    Write Partition to Ribbon ID=::NewEnumerator0
   )
   &InitializeParticle(
    Color=&MaskLinearColorBySpawnGroup(Masked Linear Color=<rgba 1 0.0264 0.4517 1>, Passthrough Linear Color=<rgba 0.0625 1.6192 2 1>, Spawn Group Mask=1)
    Lifetime=0.5
    Ribbon Width=&FloatFromCurve(
     FloatCurve=DataInterface() {propertyValues {Curve {keys [<at 0 0.2>,<at 0.1 0.7>,<at 1 1.25>]}}}
     CurveIndex=$Particles.PartitionParticles.Normalized Index In Partition
     Scale Curve=40
    )
    Ribbon Width Mode=::NewEnumerator1
    Sprite Size Mode=::NewEnumerator3
    Uniform Sprite Size=50
    UsePositionOffset=false
    Write Material Random=true
   )
   /SetVariables(Particles.IndexCurve:NiagaraFloat=&FloatFromCurve(
     FloatCurve=DataInterface() {propertyValues {Curve {
        defaultValue=0
        keys [
         {arriveTangent=0.1216, interpMode=RCIM_Cubic, leaveTangent=0.1216, tangentMode=RCTM_User, time=0, value=0}
         {arriveTangent=0.7556, interpMode=RCIM_Cubic, leaveTangent=0.7556, tangentMode=RCTM_User, time=0.25, value=0.15}
         {arriveTangent=1.465, tangentMode=RCTM_User, time=1, value=1}
        ]
       }}}
     CurveIndex=$Particles.PartitionParticles.Normalized Index In Partition
    )) {isSetParameters=true}
  ]
  particleUpdateScript [
   &ParticleState()
   /SetVariables(Particles.Pulse:NiagaraFloat=&FloatFromCurve(
     FloatCurve=DataInterface() {propertyValues {Curve {
        defaultValue=1
        keys [
         {arriveTangent=10.3786, interpMode=RCIM_Cubic, leaveTangent=10.3786, tangentMode=RCTM_User, time=0, value=0}
         {interpMode=RCIM_Cubic, tangentMode=RCTM_User, time=0.2, value=1}
         {arriveTangentWeight=0.1286, interpMode=RCIM_Cubic, leaveTangentWeight=0.2527, tangentMode=RCTM_User, time=0.4, value=0}
        ]
       }}}
     CurveIndex=&Add_Float(
      A=&Multiply_Float(A=$Particles.PartitionParticles.Normalized Index In Partition, B=-0.5)
      B=$Particles.NormalizedAge
     )
     Scale Curve=&FloatFromCurve(
      FloatCurve=DataInterface() {propertyValues {Curve {
         defaultValue=1
         keys [
          {arriveTangent=23.8449, interpMode=RCIM_Cubic, leaveTangent=23.8449, tangentMode=RCTM_User, time=0, value=0}
          {interpMode=RCIM_Cubic, tangentMode=RCTM_User, time=0.1, value=1}
          {arriveTangentWeight=0.3291, interpMode=RCIM_Cubic, leaveTangentWeight=0.1194, tangentMode=RCTM_User, time=0.7604, value=1}
          {arriveTangent=-3.1649, tangentMode=RCTM_User, time=1, value=0.5}
         ]
        }}}
      CurveIndex=$Particles.PartitionParticles.Normalized Index In Partition
     )
    )) {isSetParameters=true, module=SetVariables_#4}
   /SetVariables(Particles.Position=&AddVectorToPosition(
     Position=&AddVectorToPosition(
      Position=&Lerp_Vector(A=Linked() {linkedVariable {name=System.LocalTreePosition, type=NiagaraPosition}}, B=<-500 0 500>, Alpha=$Particles.IndexCurve)
      Vector=&VectorFromCurve(
       VectorCurve=DataInterface() {propertyValues {
         XCurve {keys [{arriveTangent=-3.0139, interpMode=RCIM_Cubic, time=0, value=0}, <at 1 0>]}
         YCurve {keys [
           {arriveTangent=-25.5959, interpMode=RCIM_Cubic, leaveTangent=-25.5959, tangentMode=RCTM_User, time=0, value=0}
           {arriveTangent=0.352, interpMode=RCIM_Cubic, leaveTangent=0.352, tangentMode=RCTM_User, time=0.0992, value=-1}
           {arriveTangent=-3.1556, interpMode=RCIM_Cubic, leaveTangent=-3.1556, tangentMode=RCTM_User, time=0.8956, value=0.2534}
           {interpMode=RCIM_Cubic, leaveTangent=1.4136, time=1, value=0}
          ]}
         ZCurve {keys [
           {arriveTangent=3.272, interpMode=RCIM_Cubic, leaveTangent=3.272, tangentMode=RCTM_User, time=0, value=0}
           {arriveTangent=-4.2899, interpMode=RCIM_Cubic, leaveTangent=-4.2899, tangentMode=RCTM_User, time=0.0877, value=0.0075}
           {arriveTangent=-0.0082, interpMode=RCIM_Cubic, leaveTangent=-0.0082, time=0.735, value=-1}
           {arriveTangent=2.5601, interpMode=RCIM_Cubic, leaveTangent=2.5601, tangentMode=RCTM_User, time=1, value=0}
          ]}
        }}
       CurveIndex=$Particles.IndexCurve
       Scale Curve=<500 1500 2000>
      )
     )
     Vector=&RotateVectorByQuaternion(
      Vector To Rotate=&Multiply_VectorByFloat(
       Vector=$System.SideDirection
       Float=&Add_Float(
        A=&FloatFromCurve(
         FloatCurve=DataInterface() {propertyValues {Curve {keys [
             {arriveTangent=17.2518, interpMode=RCIM_Cubic, leaveTangent=17.2518, tangentMode=RCTM_User, time=0, value=0}
             {arriveTangentWeight=0.0174, interpMode=RCIM_Cubic, leaveTangentWeight=0.0242, tangentMode=RCTM_User, time=0.05, value=0.4}
             {arriveTangentWeight=0.0983, interpMode=RCIM_Cubic, leaveTangentWeight=0.843, tangentMode=RCTM_User, time=0.25, value=0.3}
             <at 1 1 RCIM_Cubic>
            ]}}}
         CurveIndex=$Particles.PartitionParticles.Normalized Index In Partition
         Scale Curve=800
        )
        B=&Multiply_Float(A=$Particles.Pulse, B=100)
       )
      )
      Quaternion=&MakeQuaternion(
       Quaternion From=0
       Axis=$System.ForwardDirection
       Angle=&Add_Float(
        A=&Multiply_FloatByInt(Float=0.33, Integer=$Particles.PartitionParticles.Partition)
        B=&Add_Float(
         A=&Multiply_Float(A=$Particles.Age, B=0.25)
         B=&FloatFromCurve(
          FloatCurve=DataInterface() {propertyValues {Curve {
             defaultValue=1
             keys [
              {arriveTangent=6.6202, interpMode=RCIM_Cubic, leaveTangent=6.6202, tangentMode=RCTM_User, time=0, value=0}
              {arriveTangent=0.4573, leaveTangent=0.4573, time=0.2053, value=0.7}
              {tangentMode=RCTM_User, time=1, value=1}
             ]
            }}}
          CurveIndex=$Particles.PartitionParticles.Normalized Index In Partition
          Scale Curve=&MaskFloat(Mask Source=0, Mask Parameter=?, Mask=$Particles.PartitionParticles.Partition, Masked Float=2, Passthrough Float=4)
         )
        )
       )
       From=?
       To=?
       Angle Type=::NewEnumerator1
       Normalize Vectors=false
      )
     )
    )) {isSetParameters=true, module=SetVariables_#5}
   &self:ScratchModule_01(EndPos:NiagaraPosition=$User.EndPoint, StartPos:NiagaraPosition=$Engine.Owner.Position) {module=ScratchModule_01}
   &DynamicMaterialParameters(
    Index 0 Param 1=$Particles.ScratchModule_01.X
    Index 0 Param 2=$Particles.ScratchModule_01.Y
    Index 0 Param 3=$Particles.ScratchModule_01.Z
    Index 0 Param 4=&Add_Float(A=$Particles.ScratchModule_01.W, B=1)
    Index 1 Param 1=$Particles.MaterialRandom
    Index 1 Param 2=&FloatFromCurve(FloatCurve=DataInterface() {propertyValues {Curve {
        defaultValue=1
        keys [{arriveTangent=1.8382, interpMode=RCIM_Cubic, leaveTangent=1.8382, tangentMode=RCTM_User, time=0, value=0.1}, <at 1 1 RCIM_Cubic>]
       }}})
    Index 1 Param 3=4
    Index 1 Param 4=&Lerp_Float(
     A=1
     B=0
     Alpha=&Min_Float(
      A:NiagaraFloat=&FloatFromCurve(
       FloatCurve=DataInterface() {propertyValues {Curve {
          defaultValue=1
          keys [
           {arriveTangent=21.7082, interpMode=RCIM_Cubic, leaveTangent=21.7082, tangentMode=RCTM_User, time=0, value=0.5}
           {arriveTangentWeight=0.0498, interpMode=RCIM_Cubic, leaveTangentWeight=0.3986, tangentMode=RCTM_User, time=0.05, value=1}
           {interpMode=RCIM_Cubic, leaveTangentWeight=0.05, tangentMode=RCTM_User, time=0.9, value=1}
           {arriveTangent=-14.4395, tangentMode=RCTM_User, time=1, value=0}
          ]
         }}}
       CurveIndex=$Particles.PartitionParticles.Normalized Index In Partition
      )
      B:NiagaraFloat=&FloatFromCurve(FloatCurve=DataInterface() {propertyValues {Curve {
          defaultValue=1
          keys [
           {interpMode=RCIM_Cubic, tangentMode=RCTM_User, time=0.7, value=1}
           {
            arriveTangent=-2.9651
            arriveTangentWeight=0.1286
            interpMode=RCIM_Cubic
            leaveTangent=-2.9651
            leaveTangentWeight=0.2527
            tangentMode=RCTM_User
            time=1
            value=0
           }
          ]
         }}})
     )
    )
    Module.Index0Float0Write=true
    Module.Index0Float1Write=true
    Module.Index0Float2Write=true
    Module.Index0Float3Write=true
    Module.Index1Float0Write=true
    Module.Index1Float1Write=true
    Module.Index1Float2Write=true
    Module.Index1Float3Write=true
    Write Parameter Index 1=::NewEnumerator1
   )
   &ScaleColor(Scale RGB=&Lerp_Vector(A=<1 1 1>, B=<5 0.5 1>, Alpha=$Particles.Pulse), ScaleA=false, ScaleRGB=true)
   &ScaleRibbonWidth(Ribbon Width Scale=&Lerp_Float(A=1, B=4, Alpha=$Particles.Pulse))
   /SetVariables(
    Particles.Airglow_Color:LinearColor=&MultiplyLinearColor(Linear Color=$Particles.Color, Scale Factor=<1.2 0.7 1 0.25>)
    Particles.Airglow_Width:NiagaraFloat=&Multiply_Float(A=$Particles.RibbonWidth, B=3)
   ) {isSetParameters=true, module=SetVariables_#6}
  ]
  rendererClasses [&NiagaraSpriteRendererProperties, &NiagaraRibbonRendererProperties]
  renderers [
   {rendererClass=&NiagaraSpriteRendererProperties}
   {rendererClass=&NiagaraRibbonRendererProperties, rendererIndex=1}
   {rendererClass=&NiagaraRibbonRendererProperties, rendererIndex=2}
  ]
 }
 Flare {
  bEnabled=false
  emitterUpdateScript [
   &EmitterState()
   &SpawnBurst_Instantaneous(Spawn Time=1.25, Use Spawn Probability=false, UseLoopCountLimit=false) {module=SpawnBurst_Instantaneous003}
   &SpawnBurst_Instantaneous(Spawn Time=1.5, Use Spawn Probability=false, UseLoopCountLimit=false)
   &SpawnBurst_Instantaneous(Spawn Time=1.75, Use Spawn Probability=false, UseLoopCountLimit=false) {module=SpawnBurst_Instantaneous002}
   &SpawnBurst_Instantaneous(Spawn Group=1, Spawn Time=1.87, Use Spawn Probability=false, UseLoopCountLimit=false) {module=SpawnBurst_Instantaneous001}
  ]
  particleSpawnScript [&InitializeParticle(
    Color=<rgba 10 0.4245 4.3651 1>
    Lifetime=&MaskFloatBySpawnGroup(Masked Float=0.12, Passthrough Float=0.25)
    Sprite Size Mode=::NewEnumerator3
    Uniform Sprite Size=&MaskFloatBySpawnGroup(Masked Float=800, Passthrough Float=2500)
    UsePositionOffset=false
    Write Material Random=true
   )]
  particleUpdateScript [
   &ParticleState()
   &ScaleSpriteSize(Uniform Curve Sprite Scale=DataInterface() {propertyValues {Curve {keys [
        {arriveTangent=5.2794, interpMode=RCIM_Cubic, leaveTangent=5.2794, tangentMode=RCTM_User, time=0, value=1}
        {arriveTangent=0.7472, interpMode=RCIM_Cubic, leaveTangent=0.7472, tangentMode=RCTM_User, time=1, value=4}
       ]}}})
   /SetVariables(Particles.WideFlareSize:Vector2f=&Vector2DFromCurve(
     Vector2Curve=DataInterface() {propertyValues {
       XCurve {keys [
         {arriveTangent=4.673, interpMode=RCIM_Cubic, leaveTangent=4.673, tangentMode=RCTM_User, time=0, value=0}
         {arriveTangentWeight=0.1536, interpMode=RCIM_Cubic, leaveTangentWeight=0.1152, tangentMode=RCTM_User, time=0.3, value=1}
         {arriveTangent=-3.1914, interpMode=RCIM_Cubic, leaveTangent=-3.1914, tangentMode=RCTM_User, time=0.5, value=0.4114}
         {arriveTangent=-2.3891, interpMode=RCIM_Cubic, leaveTangent=-2.3891, tangentMode=RCTM_User, time=0.7, value=0.5114}
         {interpMode=RCIM_Cubic, leaveTangent=-1.231, time=1, value=0}
        ]}
       YCurve {keys [
         {arriveTangent=10.6903, interpMode=RCIM_Cubic, leaveTangent=10.6903, tangentMode=RCTM_User, time=0, value=0}
         {arriveTangent=1.3228, interpMode=RCIM_Cubic, leaveTangent=1.3228, time=0.25, value=1}
         {arriveTangent=-1.7663, interpMode=RCIM_Cubic, leaveTangent=-1.7663, tangentMode=RCTM_User, time=0.5, value=0.6614}
         {arriveTangent=-4.8806, interpMode=RCIM_Cubic, leaveTangent=-4.8806, tangentMode=RCTM_User, time=0.7, value=1}
         {interpMode=RCIM_Cubic, leaveTangent=-2.2148, time=1, value=0}
        ]}
      }}
     ScaleCurve=&MaskVector2dBySpawnGroup(Masked Vector2D=<8000 3000>, Passthrough Vector2D=<12000 6000>, Spawn Group Mask=0)
    )) {isSetParameters=true}
   &DynamicMaterialParameters(
    Index 0 Param 1=1
    Index 0 Param 2=&Lerp_Float(
     A=0.2
     B=1.5
     Alpha=&FloatFromCurve(FloatCurve=DataInterface() {propertyValues {Curve {keys [
          {arriveTangent=1.5591, interpMode=RCIM_Cubic, leaveTangent=1.5591, tangentMode=RCTM_User, time=0, value=0}
          {arriveTangent=0.6818, interpMode=RCIM_Cubic, leaveTangent=0.6818, tangentMode=RCTM_User, time=1, value=1}
         ]}}})
    )
    Index 3 Param 1=2
    Index 3 Param 2=5
    Module.Index0Float0Write=true
    Module.Index0Float1Write=true
    Module.Index0Float2Write=true
    Module.Index0Float3Write=false
    Module.Index3Float0Write=true
    Module.Index3Float1Write=true
    Module.Index3Float2Write=false
    Module.Index3Float3Write=false
    Write Parameter Index 3=::NewEnumerator1
   )
   &ScaleColor(
    Scale Alpha=&FloatFromCurve(FloatCurve=DataInterface() {propertyValues {Curve {keys [
         {arriveTangent=14.886, interpMode=RCIM_Cubic, leaveTangent=14.886, tangentMode=RCTM_User, time=0, value=0}
         {arriveTangent=-2.4912, interpMode=RCIM_Cubic, leaveTangent=-2.4912, tangentMode=RCTM_User, time=0.1, value=1}
         <at 1 0>
        ]}}})
    ScaleA=true
    ScaleRGB=false
   )
  ]
  rendererClasses [&NiagaraSpriteRendererProperties]
  renderers [
   {rendererClass=&NiagaraSpriteRendererProperties}
   {rendererClass=&NiagaraSpriteRendererProperties, rendererIndex=1}
  ]
 }
 TetherMotes {
  emitterUpdateScript [
   &EmitterState()
   &SpawnParticlesFromOtherEmitter(
    Attribute Reader=DataInterface() {propertyValues {EmitterBinding {bindingMode=Other, emitterName=TetherRibbons}}}
    Spawn Rate=&FloatFromCurve(
     FloatCurve=DataInterface() {propertyValues {Curve {keys [<at 0 1 RCIM_Cubic>,<at 1 0>]}}}
     CurveIndex=$Emitter.NormalizedLoopAge
     Scale Curve=25
    )
    Spawn Rate Per Particle Cap=25
    Use Maximum Age=false
    Use Minimum Age=false
   )
  ]
  particleSpawnScript [
   &InitializeParticle(
    Color Maximum=<rgba 2.5 0.8289 1.9131 1>
    Color Minimum=<rgba 1.5 0.0156 0.7344 1>
    Color Mode=::NewEnumerator2
    Lifetime Max=1
    Lifetime Min=0.5
    Lifetime Mode=::NewEnumerator1
    Sprite Size Mode=::NewEnumerator1
    Uniform Sprite Size Max=22
    Uniform Sprite Size Min=8
    UsePositionOffset=false
    Write Material Random=true
   )
   &SampleParticlesFromOtherEmitter()
   &AddVelocity(
    Velocity=&self:GetRibbonDirection(
     Particle Attribute Reader:NiagaraDataInterfaceParticleRead=DataInterface() {propertyValues {EmitterBinding {bindingMode=Other, emitterName=TetherRibbons}}}
     Index:NiagaraInt32=&MakeIntFromFloat(Float=$Particles.SampleParticlesFromOtherEmitter.SampledIndex)
     IndexOffset:NiagaraInt32=1
     FallBackVelocity:Vector3f=&Multiply_VectorByFloat(
      Vector=&TransformVector(
       Vector=&NormalizeVector(
        Vector To Normalize=&Subtract_Position(A=$User.StartPosition, B=&AddVectorToPosition(Position=$Engine.Owner.Position, Vector=<0 0 4500>))
        Safe Normalize=true
        Safe Fallback Threshold=0
        Safe Fallback Vector=<0 0 0.001>
       )
       Source Space=::World
       Destination Space=::Simulation
      )
      Float=-500
     )
    )
    Velocity Speed Scale=&RandomRangeFloat(Minimum=2, Maximum=4)
   )
   &ShapeLocation(MaskSpawnGroup=false, Override Seed=false, Sphere Radius=150, Sphere Surface Distribution=1, UseOwnerScale=true)
  ]
  particleUpdateScript [
   &ParticleState()
   &ScaleSpriteSize(
    Uniform Curve Scale=&FloatFromCurve(FloatCurve=DataInterface() {propertyValues {Curve {keys [<at 6000 0.5>,<at 9000 1>,<at 15000 3>]}}}, CurveIndex=$Engine.Owner.LODDistance)
    Uniform Curve Sprite Scale=DataInterface() {propertyValues {Curve {
       defaultValue=0
       keys [
        {arriveTangent=5.1504, interpMode=RCIM_Cubic, leaveTangent=5.1504, tangentMode=RCTM_User, time=0, value=0}
        {interpMode=RCIM_Cubic, tangentMode=RCTM_User, time=0.5, value=1}
        {arriveTangent=-4.4997, tangentMode=RCTM_User, time=1, value=0}
       ]
      }}}
   )
   &VectorNoiseForce(Force Amount=8000, Module.Use Random Noise Time Interval=true, Random Noise Time Interval=0.4)
   &Drag(Drag=2, Use Linear Drag=true, Use Rotational Drag=false)
   &SolveForcesAndVelocity(Clamp Velocity=false, Limit Acceleration=false)
   &ScaleSpriteSizeBySpeed(
    Max Scale Factor=<0.7 5>
    Sample Scale Factor By Curve=true
    Scale Factor Curve=DataInterface() {propertyValues {Curve {keys [<at 0 0 RCIM_Cubic>,<at 1 1>]}}}
    Velocity Threshold=2000
   )
  ]
  rendererClasses [&NiagaraSpriteRendererProperties]
  renderers [{rendererClass=&NiagaraSpriteRendererProperties}]
 }
 TetherRibbons {
  emitterUpdateScript [
   &EmitterState()
   &SpawnBurst_Instantaneous(Spawn Count=20, Use Spawn Probability=false, UseLoopCountLimit=false)
   &SpawnBurst_Instantaneous(Spawn Count=20, Spawn Time=0.2, Use Spawn Probability=false, UseLoopCountLimit=false) {module=SpawnBurst_Instantaneous001}
  ]
  particleSpawnScript [
   &PartitionParticles(
    Normalized Index In Partition Includes One=true
    Particles Per Partition=20
    Partition Mode:ENiagara_PartitionMode_v2=::NewEnumerator1
    Write Partition Index to Ribbon Link Order:ENiagara_AttributeSamplingApplyOutput=::NewEnumerator0
    Write Partition to Ribbon ID=::NewEnumerator0
   )
   &InitializeParticle(
    Color=&LerpLinearColors(
     StartColor=<rgba 0.439 4.0372 5 1>
     EndColor=<rgba 0 0.7992 5 1>
     LerpFactor=&FloatFromCurve(
      FloatCurve=DataInterface() {propertyValues {Curve {
         defaultValue=1
         keys [
          {arriveTangent=7.1971, interpMode=RCIM_Cubic, time=0, value=0}
          {arriveTangentWeight=0.0833, interpMode=RCIM_Cubic, leaveTangentWeight=0.15, tangentMode=RCTM_User, time=0.25, value=1}
          {arriveTangentWeight=0.4206, interpMode=RCIM_Cubic, leaveTangentWeight=0.2804, tangentMode=RCTM_User, time=0.7, value=1}
          <at 1 0 RCIM_Cubic>
         ]
        }}}
      CurveIndex=$Particles.PartitionParticles.Normalized Index In Partition
     )
    )
    Lifetime=2.3
    Ribbon Facing Mode=::NewEnumerator1
    Ribbon Width=&FloatFromCurve(
     FloatCurve=DataInterface() {propertyValues {Curve {keys [
         {arriveTangent=5.3098, interpMode=RCIM_Cubic, leaveTangent=5.3098, tangentMode=RCTM_User, time=0, value=0.5}
         {arriveTangent=0.5, interpMode=RCIM_Cubic, leaveTangent=0.5, time=0.8, value=2}
         {arriveTangent=-10.7185, tangentMode=RCTM_User, time=1, value=1}
        ]}}}
     CurveIndex=$Particles.PartitionParticles.Normalized Index In Partition
     Scale Curve=300
    )
    Ribbon Width Mode=::NewEnumerator1
    UsePositionOffset=false
    Write Material Random=true
   )
   /SetVariables(Particles.IndexCurve:NiagaraFloat=&FloatFromCurve(
     FloatCurve=DataInterface() {propertyValues {Curve {
        defaultValue=0
        keys [
         {arriveTangent=0.1216, interpMode=RCIM_Cubic, leaveTangent=0.1216, tangentMode=RCTM_User, time=0, value=0}
         {arriveTangent=0.7556, interpMode=RCIM_Cubic, leaveTangent=0.7556, tangentMode=RCTM_User, time=0.25, value=0.15}
         {arriveTangent=1.465, tangentMode=RCTM_User, time=1, value=1}
        ]
       }}}
     CurveIndex=$Particles.PartitionParticles.Normalized Index In Partition
    )) {isSetParameters=true}
  ]
  particleUpdateScript [
   &ParticleState()
   /SetVariables(Particles.Position=&AddVectorToPosition(
     Position=&LerpPosition(A=$System.LocalTreePosition, B=<-500 0 500>, Alpha=$Particles.PartitionParticles.Normalized Index In Partition)
     Vector=&MaskVector(
      Mask Source=1
      Mask Parameter=$Particles.PartitionParticles.Partition
      Mask=0
      Masked Vector=&VectorFromCurve(
       VectorCurve=DataInterface() {propertyValues {
         XCurve {keys [{arriveTangent=-3.0139, interpMode=RCIM_Cubic, time=0, value=0}, <at 1 0>]}
         YCurve {keys [
           {arriveTangent=-25.5959, interpMode=RCIM_Cubic, leaveTangent=-25.5959, tangentMode=RCTM_User, time=0, value=0}
           {arriveTangent=0.352, interpMode=RCIM_Cubic, leaveTangent=0.352, tangentMode=RCTM_User, time=0.0992, value=-1}
           {arriveTangent=-3.1556, interpMode=RCIM_Cubic, leaveTangent=-3.1556, tangentMode=RCTM_User, time=0.8956, value=0.2534}
           {interpMode=RCIM_Cubic, leaveTangent=1.4136, time=1, value=0}
          ]}
         ZCurve {keys [
           {arriveTangent=10.8811, interpMode=RCIM_Cubic, leaveTangent=10.8811, tangentMode=RCTM_User, time=0, value=0.1}
           {arriveTangent=-4.2899, interpMode=RCIM_Cubic, leaveTangent=-4.2899, tangentMode=RCTM_User, time=0.0877, value=0.0075}
           {arriveTangent=-0.0082, interpMode=RCIM_Cubic, leaveTangent=-0.0082, time=0.735, value=-1}
           {arriveTangent=2.5601, interpMode=RCIM_Cubic, leaveTangent=2.5601, tangentMode=RCTM_User, time=1, value=0}
          ]}
        }}
       CurveIndex=$Particles.PartitionParticles.Normalized Index In Partition
       Scale Curve=<0 2000 3000>
      )
      Passthrough Vector=&VectorFromCurve(
       VectorCurve=DataInterface() {propertyValues {
         XCurve {keys [{arriveTangent=-3.0139, interpMode=RCIM_Cubic, time=0, value=0}, {interpMode=RCIM_Cubic, leaveTangent=0.2713, time=1, value=0}]}
         YCurve {keys [
           {arriveTangent=-4.3003, interpMode=RCIM_Cubic, leaveTangent=-4.3003, tangentMode=RCTM_User, time=0, value=0}
           {arriveTangent=0.1276, interpMode=RCIM_Cubic, leaveTangent=0.1276, tangentMode=RCTM_User, time=0.4548, value=-0.8967}
           {arriveTangentWeight=0.2372, interpMode=RCIM_Cubic, leaveTangentWeight=0.15, tangentMode=RCTM_User, time=0.8666, value=1}
           {interpMode=RCIM_Cubic, leaveTangent=-1.9289, time=1.0016, value=0}
          ]}
         ZCurve {keys [
           {arriveTangent=-2.9173, interpMode=RCIM_Cubic, leaveTangent=-2.9173, tangentMode=RCTM_User, time=0, value=0}
           {arriveTangent=-3.981, interpMode=RCIM_Cubic, leaveTangent=-3.981, tangentMode=RCTM_User, time=0.2, value=-0.15}
           {arriveTangent=0.1607, interpMode=RCIM_Cubic, leaveTangent=0.1607, tangentMode=RCTM_User, time=0.8609, value=-1}
           {arriveTangent=12.0556, interpMode=RCIM_Cubic, leaveTangent=12.0556, tangentMode=RCTM_User, time=1, value=0}
          ]}
        }}
       CurveIndex=$Particles.PartitionParticles.Normalized Index In Partition
       Scale Curve=<0 2000 3000>
      )
     )
    )) {isSetParameters=true}
   &DynamicMaterialParameters(
    Index 0 Param 1=&FloatFromCurve(
     FloatCurve=DataInterface() {propertyValues {Curve {
        defaultValue=1
        keys [
         <at 0 0.5 RCIM_Cubic>
         {arriveTangentWeight=0.0236, interpMode=RCIM_Cubic, leaveTangentWeight=0.2909, tangentMode=RCTM_User, time=0.06, value=0}
        ]
       }}}
     CurveIndex=$Particles.PartitionParticles.Normalized Index In Partition
    )
    Index 0 Param 2=&MaskFloat(Mask Source=1, Mask Parameter=$Particles.PartitionParticles.Partition, Mask=0, Masked Float=6, Passthrough Float=8)
    Index 0 Param 3=&FloatFromCurve(
     FloatCurve=DataInterface() {propertyValues {Curve {
        defaultValue=1
        keys [
         {arriveTangent=0.3225, interpMode=RCIM_Cubic, leaveTangent=0.3225, tangentMode=RCTM_User, time=0, value=0}
         {arriveTangent=1.7622, interpMode=RCIM_Cubic, leaveTangent=1.7622, tangentMode=RCTM_User, time=1, value=1}
        ]
       }}}
     Scale Curve=&MaskFloat(Mask Source=1, Mask Parameter=$Particles.PartitionParticles.Partition, Mask=0, Masked Float=8, Passthrough Float=12)
    )
    Index 1 Param 3=&FloatFromCurve(FloatCurve=DataInterface() {propertyValues {Curve {defaultValue=0, keys [{tangentMode=RCTM_User, time=0, value=0}, {tangentMode=RCTM_User, time=1, value=1}]}}})
    Index 1 Param 4=1
    Module.Index0Float0Write=true
    Module.Index0Float1Write=true
    Module.Index0Float2Write=true
    Module.Index0Float3Write=false
    Module.Index1Float0Write=false
    Module.Index1Float1Write=false
    Module.Index1Float2Write=true
    Module.Index1Float3Write=true
    Write Parameter Index 1=::NewEnumerator1
   )
   &ScaleRibbonWidth(Ribbon Width Scale=&FloatFromCurve(FloatCurve=DataInterface() {propertyValues {Curve {keys [
         {arriveTangent=3.2446, interpMode=RCIM_Cubic, leaveTangent=3.2446, tangentMode=RCTM_User, time=0, value=0}
         {arriveTangent=-1.8445, interpMode=RCIM_Cubic, leaveTangent=-1.8445, tangentMode=RCTM_User, time=0.3249, value=0.8128}
         {arriveTangent=2.3297, tangentMode=RCTM_User, time=1, value=1}
        ]}}}))
  ]
  rendererClasses [&NiagaraSpriteRendererProperties, &NiagaraRibbonRendererProperties]
  renderers [
   {rendererClass=&NiagaraSpriteRendererProperties}
   {rendererClass=&NiagaraRibbonRendererProperties, rendererIndex=1}
  ]
 }
}
summary {
 systemName=NS_DCN_RealityTear_Zap
 emitters [
  {
   emitterName=TetherRibbons
   rendererClasses [&NiagaraSpriteRendererProperties, &NiagaraRibbonRendererProperties]
  }
  {emitterName=TetherMotes, rendererClasses [&NiagaraSpriteRendererProperties]}
  {bEnabled=false, emitterName=ElectricArcs, rendererClasses [&NiagaraSpriteRendererProperties]}
  {
   emitterName=ElectricityUp
   rendererClasses [&NiagaraSpriteRendererProperties, &NiagaraRibbonRendererProperties]
  }
  {
   emitterName=ElectricityDown
   rendererClasses [&NiagaraSpriteRendererProperties, &NiagaraRibbonRendererProperties]
  }
  {bEnabled=false, emitterName=Flare, rendererClasses [&NiagaraSpriteRendererProperties]}
 ]
 userVariables [
  User.EndPoint {type=NiagaraPosition, description="", defaultValue {}}
  User.StartPosition {type=NiagaraPosition, description="", defaultValue {}}
 ]
}
