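# Niagara system for a blue rift opening: a burst of cracks, flares, shards and sparkles.
# Crack meshes and a second flare persist after the short main flare and Telegraph mesh, animated through curve-driven scale and dynamic material parameters.
# ShardsRadial and SparklesRadial combine bursts with decaying spawn rates, slowed by drag; ShardsFixed, ShardsForward and ShardsOpening are disabled.
# User.Scale_FX scales many sizes and spawn shapes.
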NiagaraSystem /DragonCartNarrative/Assets/VFX/BlueRifts/NS_DCN_Rift_Open
dependencies {
 usedDataInterfaces [
  &NiagaraDataInterfaceCurve
  &NiagaraDataInterfaceVector2DCurve
  &NiagaraDataInterfaceVectorCurve
 ]
 usedDynamicInputs []
 usedModules [&SystemState]
}
emitters {
 CrackSpline {
  emitterUpdateScript [
   &EmitterState(Life Cycle Mode=::NewEnumerator1, Loop Behavior=::NewEnumerator1, Loop Duration Mode=::NewEnumerator1, UseLoopDelay=false)
   &SpawnBurst_Instantaneous(Spawn Count=2, Spawn Time=0.05, Use Spawn Probability=false)
  ]
  particleSpawnScript [
   /SetVariables(Particles.MeshOrientation=&MakeQuatFromAxisAngle(
     Axis=&RandomRangeVector(Minimum=<0 0 0>, Maximum=<1 1 1>, Evaluation Type=::NewEnumerator1)
     AngleInDegrees=90
    )) {isSetParameters=true}
   &InitializeParticle(
    Lifetime=20
    Mesh Scale Mode=::NewEnumerator1
    Mesh Uniform Scale Max=280
    Mesh Uniform Scale Min=100
    UsePositionOffset=false
    Write Material Random=true
   ) {module=InitializeParticle001}
   &Color(Color=&MakeLinearColorFromVectorAndFloat("Vector (RGB)"=<1.5 1.5 1.5>, "Float (Alpha)"=&ClampFloat(Float=1)))
   &SystemLocation(MaskSpawnGroup=false, Offset=<800 0 0>)
  ]
  particleUpdateScript [
   &ParticleState(Kill Particles When Lifetime Has Elapsed=false, Let Infinitely Lived Particles Die When Emitter Deactivates=true)
   &ScaleMeshSize(Scale Factor=&VectorFromFloat(Value=$User.Scale_FX))
   &ScaleColor(
    Scale Alpha=&FloatFromCurve(FloatCurve=DataInterface() {propertyValues {Curve {keys [<at 0.77 1 RCIM_Cubic>,<at 1 0>]}}})
    Scale RGB=&VectorFromCurve(
     VectorCurve=DataInterface() {propertyValues {
       XCurve {keys [
         {arriveTangent=-6.9225, interpMode=RCIM_Cubic, time=0, value=4.5}
         {arriveTangent=-3.6158, interpMode=RCIM_Cubic, leaveTangent=-3.6158, time=0.65, value=1}
         <at 1 0.8842 RCIM_Cubic>
        ]}
       YCurve {keys [
         {arriveTangent=-15.8226, interpMode=RCIM_Cubic, time=0, value=6}
         {arriveTangent=-5.6986, interpMode=RCIM_Cubic, leaveTangent=-5.6986, time=0.65, value=1}
         {interpMode=RCIM_Cubic, leaveTangent=-5.843, time=1, value=0.3014}
        ]}
       ZCurve {keys [
         {arriveTangent=-15.2224, interpMode=RCIM_Cubic, time=0, value=6}
         {arriveTangent=-5, interpMode=RCIM_Cubic, leaveTangent=-5, time=0.65, value=1}
         <at 1 1 RCIM_Cubic>
        ]}
      }}
     Scale Curve=<0.12 0.12 0.12>
    )
    ScaleA=true
    ScaleRGB=true
   ) {enabled=false}
   &DynamicMaterialParameters(
    Index 0 Param 1=&FloatFromCurve(FloatCurve=DataInterface() {propertyValues {Curve {keys [
         <at 0 1 RCIM_Cubic>
         {arriveTangent=-1, interpMode=RCIM_Cubic, leaveTangent=-1, time=0.3329, value=0.616}
         {interpMode=RCIM_Cubic, leaveTangent=-0.3905, time=1, value=0}
        ]}}})
    Index 0 Param 2=&FloatFromCurve(
     FloatCurve=DataInterface() {propertyValues {Curve {keys [
         {arriveTangent=-2.0225, interpMode=RCIM_Cubic, time=0, value=0.6}
         {arriveTangent=-0.6, interpMode=RCIM_Cubic, leaveTangent=-0.6, time=0.65, value=0.25}
         <at 1 0 RCIM_Cubic>
        ]}}}
     Scale Curve=&Multiply_Float(A=20, B=$User.Scale_FX)
    )
    Index 0 Param 3=&FloatFromCurve(FloatCurve=DataInterface() {propertyValues {Curve {keys [{arriveTangent=1.0145, interpMode=RCIM_Cubic, time=0, value=-0.1}, <at 0.25 0>]}}})
    Index 0 Param 4=&FloatFromCurve(FloatCurve=DataInterface() {propertyValues {Curve {keys [<at 0 0 RCIM_Cubic>,<at 1 0>]}}})
    Module.Param0WriteEnabled=true
    Module.Param1WriteEnabled=true
    Module.Param2WriteEnabled=true
    Module.Param3WriteEnabled=true
   )
   &CameraOffset(Camera Offset Amount=-100)
  ]
  rendererClasses [&NiagaraMeshRendererProperties]
  renderers [{rendererClass=&NiagaraMeshRendererProperties, rendererIndex=0}]
 }
 Flare {
  emitterUpdateScript [
   &EmitterState()
   &SpawnBurst_Instantaneous(Spawn Time=0.1, Use Spawn Probability=false, UseLoopCountLimit=false)
  ]
  particleSpawnScript [
   &InitializeParticle(
    Color=<rgba 0.4977 1.8664 8 1>
    Lifetime=1.25
    Sprite Size Mode=::NewEnumerator3
    Uniform Sprite Size=&Multiply_Float(A=10000, B=$User.Scale_FX)
    UsePositionOffset=false
    Write Material Random=true
   )
   &SystemLocation(MaskSpawnGroup=false, Offset=<800 0 0>)
  ]
  particleUpdateScript [
   &ParticleState()
   &ScaleSpriteSize(Uniform Curve Sprite Scale=DataInterface() {propertyValues {Curve {keys [
        {arriveTangent=5.2794, interpMode=RCIM_Cubic, leaveTangent=5.2794, tangentMode=RCTM_User, time=0, value=1}
        {arriveTangent=0.7472, interpMode=RCIM_Constant, time=0.6, value=4}
        <at 0.61 0 RCIM_Constant>
       ]}}})
   &CameraOffset(Camera Offset Amount=&Lerp_Float(A=2500, B=5000, Alpha=$Particles.NormalizedAge))
   /SetVariables(Particles.GlowCamOffset=7000) {isSetParameters=true}
   /SetVariables(Particles.WideFlareSize:Vector2f=&Vector2DFromCurve(
     Vector2Curve=DataInterface() {propertyValues {
       XCurve {keys [
         {arriveTangent=4.673, interpMode=RCIM_Cubic, leaveTangent=4.673, tangentMode=RCTM_User, time=0, value=0}
         {arriveTangentWeight=0.1536, interpMode=RCIM_Cubic, leaveTangentWeight=0.1152, tangentMode=RCTM_User, time=0.3, value=1}
         {arriveTangent=-3.1914, interpMode=RCIM_Cubic, leaveTangent=-3.1914, tangentMode=RCTM_User, time=0.5, value=0.15}
         {arriveTangent=-2.3891, interpMode=RCIM_Cubic, leaveTangent=-2.3891, tangentMode=RCTM_User, time=0.7, value=0.25}
         {interpMode=RCIM_Cubic, leaveTangent=-1.231, time=1, value=0}
        ]}
       YCurve {keys [
         {arriveTangent=10.6903, interpMode=RCIM_Cubic, leaveTangent=10.6903, tangentMode=RCTM_User, time=0, value=0}
         {arriveTangent=0.8, interpMode=RCIM_Cubic, leaveTangent=0.8, time=0.25, value=1}
         {arriveTangent=-1.7663, interpMode=RCIM_Cubic, leaveTangent=-1.7663, tangentMode=RCTM_User, time=0.5, value=0.4}
         {arriveTangent=-4.8806, interpMode=RCIM_Cubic, leaveTangent=-4.8806, tangentMode=RCTM_User, time=0.7, value=1}
         {interpMode=RCIM_Cubic, leaveTangent=-2.2148, time=1, value=0}
        ]}
      }}
     ScaleCurve=&Multiply_Vector2D_ByFloat(Float=$User.Scale_FX, Vector2D=<100000 14000>)
    )) {isSetParameters=true}
   /SetVariables(Particles.GlowSize=&Vector2DFromFloat(Value=&Multiply_Float(A=75000, B=$User.Scale_FX))) {isSetParameters=true}
   &DynamicMaterialParameters(
    Index 0 Param 1=1
    Index 0 Param 2=&Lerp_Float(
     A=2
     B=0.25
     Alpha=&FloatFromCurve(FloatCurve=DataInterface() {propertyValues {Curve {keys [
          {arriveTangent=3.9351, interpMode=RCIM_Cubic, leaveTangent=3.9351, tangentMode=RCTM_User, time=0, value=0}
          {interpMode=RCIM_Cubic, leaveTangent=0.3903, time=0.6, value=1}
         ]}}})
    )
    Index 3 Param 1=2
    Index 3 Param 2=5
    Module.Index0Param0Write=true
    Module.Index0Param1Write=true
    Module.Index0Param2Write=true
    Module.Index0Param3Write=false
    Module.Index3Param0Write=true
    Module.Index3Param1Write=true
    Module.Index3Param2Write=false
    Module.Index3Param3Write=false
    Write Parameter Index 3=true
   )
   &ScaleColor(
    Scale Alpha=&FloatFromCurve(FloatCurve=DataInterface() {propertyValues {Curve {keys [<at 0 1 RCIM_Cubic>,<at 0.5 0>]}}})
    ScaleA=true
    ScaleRGB=false
   )
   /SetVariables(Particles.GlowColor=&MultiplyLinearColor(
     Linear Color=<rgba 0.1307 0.3621 1 1>
     Scale Factor=&MakeFloat4FromVectorAndFloat(
      "Vector (RGB)"=<1 1 1>
      "Float (Alpha)"=&FloatFromCurve(
       FloatCurve=DataInterface() {propertyValues {Curve {keys [
           {arriveTangent=16.7285, interpMode=RCIM_Cubic, leaveTangent=16.7285, tangentMode=RCTM_User, time=0, value=0.35}
           {arriveTangentWeight=0.0353, interpMode=RCIM_Cubic, leaveTangentWeight=0.3178, tangentMode=RCTM_User, time=0.1, value=1}
           {arriveTangent=-0.0405, interpMode=RCIM_Cubic, leaveTangent=-0.0405, tangentMode=RCTM_User, time=1, value=0}
          ]}}}
       Scale Curve=0.25
      )
     )
    )) {isSetParameters=true}
   /SetVariables(Particles.Position=&MakeVector(X=0, Y=0, Z=0)) {isSetParameters=true}
  ]
  rendererClasses [&NiagaraSpriteRendererProperties]
  renderers [
   {rendererClass=&NiagaraSpriteRendererProperties}
   {rendererClass=&NiagaraSpriteRendererProperties, rendererIndex=1}
   {rendererClass=&NiagaraSpriteRendererProperties, rendererIndex=2}
  ]
 }
 Flare001 {
  emitterUpdateScript [
   &EmitterState(Life Cycle Mode=::NewEnumerator1, Loop Behavior=::NewEnumerator1, Loop Duration Mode=::NewEnumerator1, UseLoopDelay=false)
   &SpawnBurst_Instantaneous(Use Spawn Probability=false, UseLoopCountLimit=false)
  ]
  particleSpawnScript [
   &InitializeParticle(
    Color=<rgba 0.1123 0.6668 1 1>
    Lifetime=20
    Sprite Size Mode=::NewEnumerator3
    Uniform Sprite Size=&Multiply_Float(A=2500, B=$User.Scale_FX)
    UsePositionOffset=false
    Write Material Random=true
   )
   &SystemLocation(MaskSpawnGroup=false, Offset=<800 0 0>)
  ]
  particleUpdateScript [
   &ParticleState(Kill Particles When Lifetime Has Elapsed=false, Let Infinitely Lived Particles Die When Emitter Deactivates=true)
   &ScaleSpriteSize(
    Uniform Curve Index=$Particles.Age
    Uniform Curve Sprite Scale=DataInterface() {propertyValues {Curve {keys [
        {arriveTangent=32.0001, interpMode=RCIM_Cubic, leaveTangent=32.0001, tangentMode=RCTM_User, time=0, value=0}
        {arriveTangent=-19.4905, interpMode=RCIM_Cubic, leaveTangent=-0.3418, tangentMode=RCTM_Break, time=0.25, value=4}
        {arriveTangentWeight=0.1374, interpMode=RCIM_Cubic, leaveTangentWeight=0.8933, tangentMode=RCTM_User, time=0.6965, value=4.3084}
       ]}}}
   )
   &CameraOffset(Camera Offset Amount=&Lerp_Float(A=1500, B=3000, Alpha=$Particles.NormalizedAge))
   &DynamicMaterialParameters(
    Index 0 Param 1=1
    Index 0 Param 2=&Lerp_Float(
     A=1
     B=0
     Alpha=&FloatFromCurve(
      FloatCurve=DataInterface() {propertyValues {Curve {keys [
          {arriveTangent=3.8893, interpMode=RCIM_Cubic, leaveTangent=3.8893, leaveTangentWeight=0.1, tangentMode=RCTM_User, time=0, value=0}
          {
           arriveTangent=12.6638
           arriveTangentWeight=0.2687
           interpMode=RCIM_Cubic
           leaveTangent=12.6638
           leaveTangentWeight=0.3583
           tangentMode=RCTM_User
           time=0.44
           value=3
          }
          {arriveTangent=2.6873, interpMode=RCIM_Cubic, leaveTangent=2.6873, tangentMode=RCTM_User, time=0.45, value=0}
          {arriveTangentWeight=0.1, interpMode=RCIM_Cubic, leaveTangentWeight=0.8933, tangentMode=RCTM_User, time=0.75, value=0.7}
         ]}}}
      CurveIndex=$Particles.Age
     )
    )
    Index 3 Param 1=2
    Index 3 Param 2=5
    Module.Index0Param0Write=true
    Module.Index0Param1Write=true
    Module.Index0Param2Write=true
    Module.Index0Param3Write=false
    Module.Index3Param0Write=true
    Module.Index3Param1Write=true
    Module.Index3Param2Write=false
    Module.Index3Param3Write=false
    Write Parameter Index 3=true
   )
  ]
  rendererClasses [&NiagaraSpriteRendererProperties]
  renderers [{rendererClass=&NiagaraSpriteRendererProperties}]
 }
 LargeCracks {
  emitterUpdateScript [
   &EmitterState(Life Cycle Mode=::NewEnumerator1, Loop Behavior=::NewEnumerator1, Loop Duration Mode=::NewEnumerator1, UseLoopDelay=false)
   &SpawnBurst_Instantaneous(Spawn Time=0.02, Use Spawn Probability=false, UseLoopCountLimit=false)
  ]
  particleSpawnScript [
   &InitializeParticle(
    Color=<rgba 1 1.5337 2 1>
    Lifetime=20
    Mesh Scale Mode=::NewEnumerator3
    Mesh Uniform Scale=&Multiply_Float(A=200, B=$User.Scale_FX)
    Position Mode=::NewEnumerator0
    Write Material Random=true
   ) {module=InitializeParticle001}
   &InitialMeshOrientation(Mesh Axis To Orient=<0 0 1>, Mesh Orientation Mode=::NewEnumerator2, Rotation=<0 0 0.4>, Use Rotation Vector=true)
  ]
  particleUpdateScript [
   &ParticleState(Kill Particles When Lifetime Has Elapsed=false, Let Infinitely Lived Particles Die When Emitter Deactivates=true)
   &ScaleColor(
    Scale Alpha=&FloatFromCurve(FloatCurve=DataInterface() {propertyValues {Curve {keys [<at 0.77 1 RCIM_Cubic>,<at 1 0>]}}})
    Scale RGB=&VectorFromFloat(Value=&FloatFromCurve(FloatCurve=DataInterface() {propertyValues {Curve {keys [
          {arriveTangent=-88.7987, interpMode=RCIM_Cubic, leaveTangent=-88.7987, tangentMode=RCTM_User, time=0, value=10}
          <at 0.08 1>
          {arriveTangentWeight=0.3005, interpMode=RCIM_Cubic, leaveTangentWeight=0.3005, tangentMode=RCTM_User, time=0.7, value=1}
          {arriveTangent=-1.1457, interpMode=RCIM_Cubic, leaveTangent=-1.1457, tangentMode=RCTM_User, time=1, value=0.4}
         ]}}}))
    ScaleA=true
    ScaleRGB=true
   ) {enabled=false}
   &ScaleMeshSize(Scale Factor=&VectorFromCurve(VectorCurve=DataInterface() {propertyValues {
       XCurve {keys [
         {arriveTangent=2.6567, interpMode=RCIM_Cubic, leaveTangent=2.6567, tangentMode=RCTM_User, time=0, value=0.9}
         <at 0.07 1>
         {arriveTangentWeight=0.1667, interpMode=RCIM_Cubic, leaveTangentWeight=0.1, tangentMode=RCTM_User, time=0.77, value=1}
        ]}
       YCurve {keys [
         {arriveTangent=2.6567, interpMode=RCIM_Cubic, leaveTangent=2.6567, tangentMode=RCTM_User, time=0, value=0.9}
         <at 0.07 1>
         {arriveTangentWeight=0.1718, interpMode=RCIM_Cubic, leaveTangentWeight=0.1031, tangentMode=RCTM_User, time=0.77, value=1}
        ]}
       ZCurve {keys [
         {arriveTangent=11.6068, interpMode=RCIM_Cubic, leaveTangent=11.6068, tangentMode=RCTM_User, time=0, value=0.7}
         {arriveTangentWeight=0.0363, interpMode=RCIM_Cubic, leaveTangentWeight=0.2176, tangentMode=RCTM_User, time=0.05, value=1}
         {arriveTangentWeight=0.2049, interpMode=RCIM_Cubic, leaveTangentWeight=0.1024, tangentMode=RCTM_User, time=0.77, value=1}
        ]}
      }}))
   &DynamicMaterialParameters(
    Index 0 Param 1=&FloatFromCurve(FloatCurve=DataInterface() {propertyValues {Curve {
        defaultValue=1
        keys [
         {arriveTangent=4.1785, interpMode=RCIM_Cubic, leaveTangent=4.1785, tangentMode=RCTM_User, time=0, value=0}
         {arriveTangent=1.1179, interpMode=RCIM_Cubic, leaveTangent=1.1179, tangentMode=RCTM_User, time=0.0902, value=0.9141}
         {arriveTangent=0.0224, interpMode=RCIM_Cubic, leaveTangent=0.0224, tangentMode=RCTM_User, time=1, value=1}
        ]
       }}})
    Index 0 Param 2=&FloatFromCurve(
     FloatCurve=DataInterface() {propertyValues {Curve {
        defaultValue=1
        keys [
         {arriveTangent=4.1785, interpMode=RCIM_Cubic, leaveTangent=4.1785, tangentMode=RCTM_User, time=0, value=0}
         {arriveTangent=1.1179, interpMode=RCIM_Cubic, leaveTangent=1.1179, tangentMode=RCTM_User, time=0.0902, value=0.9141}
         {arriveTangent=0.0224, interpMode=RCIM_Cubic, leaveTangent=0.0224, tangentMode=RCTM_User, time=1, value=1}
        ]
       }}}
     Scale Curve=&Multiply_Float(A=5, B=$User.Scale_FX)
    )
    Module.Index0Float0Write=true
    Module.Index0Float1Write=true
    Module.Index0Float2Write=true
    Module.Index0Float3Write=true
    Module.Index1Float0Write=true
    Module.Index1Float1Write=false
    Module.Index1Float2Write=false
    Module.Index1Float3Write=false
    Write Parameter Index 1=::NewEnumerator1
   )
  ]
  rendererClasses [&NiagaraMeshRendererProperties]
  renderers [
   {rendererClass=&NiagaraMeshRendererProperties, rendererIndex=0}
   {rendererClass=&NiagaraMeshRendererProperties, rendererIndex=1}
  ]
 }
 ShardsFixed {
  bEnabled=false
  emitterUpdateScript [
   &EmitterState()
   &SpawnBurst_Instantaneous(Spawn Count=14, Use Spawn Probability=false, UseLoopCountLimit=false)
  ]
  particleSpawnScript [
   &InitializeParticle(
    Color Maximum=<rgba 3.2706 7.7612 12 2>
    Color Minimum=<rgba 1.5607 1.9496 6 2>
    Color Mode=::NewEnumerator2
    Lifetime=&Multiply_Float(A=$System.Duration, B=&RandomRangeFloat(Minimum=0.9, Maximum=1))
    Mesh Scale Max=<100 100 100>
    Mesh Scale Min=<40 40 40>
    Mesh Scale Mode=::NewEnumerator2
    Position Offset=<1700 0 0>
    Position Offset Coordinate Space=::Local
    Sprite Size Mode=::NewEnumerator1
    Uniform Sprite Size Max=4000
    Uniform Sprite Size Min=2500
    UsePositionOffset=true
    Write Material Random=true
   )
   /SetVariables(Particles.MeshIndex=&RandomRangeInt(Minimum=0, Maximum=2)) {isSetParameters=true, module=SetVariables_#1}
   &ShapeLocation(
    Cylinder Distribution=::NewEnumerator2
    Cylinder Height=50
    Cylinder Radius=1000
    MaskSpawnGroup=false
    Non Uniform Scale=&Multiply_VectorByFloat(Vector=<1 1 0.5>, Float=$User.Scale_FX)
    Rotation Mode=::NewEnumerator1
    Shape Primitive=::NewEnumerator1
    U Position=&ReturnNormalizedExecIndex(Normalized Index Scale=&RandomRangeFloat(Minimum=0.9, Maximum=1.1))
    UseOwnerScale=true
   )
   &InitialMeshOrientation(
    Mesh Axis To Orient=<0 -1 0>
    Mesh Orientation Mode=::NewEnumerator2
    Orientation Vector=$Particles.ShapeLocation.ShapeNormal
    Rotation=<0 0.25 0>
    Use Rotation Vector=true
   )
   &CameraOffset(Camera Offset Amount=500)
  ]
  particleUpdateScript [
   &ParticleState(Kill Particles When Lifetime Has Elapsed=false)
   &ScaleMeshSize(Scale Factor=&VectorFromFloat(Value=&FloatFromCurve(FloatCurve=DataInterface() {propertyValues {Curve {keys [
          {arriveTangentWeight=0.0236, interpMode=RCIM_Cubic, leaveTangentWeight=0.3536, tangentMode=RCTM_User, time=0.75, value=1}
          {arriveTangent=-4, interpMode=RCIM_Cubic, leaveTangent=-4, time=0.9, value=0.8}
          {arriveTangent=-10.5327, tangentMode=RCTM_User, time=1, value=0}
         ]}}})))
   /SetVariables(Particles.Position=&AddVectorToPosition(
     Position=$Particles.Initial.Position
     Vector=&Multiply_VectorByFloat(
      Vector=Linked() {linkedVariable {name=Particles.Initial.Position, type=NiagaraPosition}}
      Float=&FloatFromCurve(
       FloatCurve=DataInterface() {propertyValues {Curve {keys [
           {arriveTangent=34.2105, interpMode=RCIM_Cubic, leaveTangent=34.2105, tangentMode=RCTM_User, time=0, value=0}
           {tangentMode=RCTM_User, time=0.05, value=1}
           {arriveTangentWeight=0.4237, interpMode=RCIM_Cubic, leaveTangentWeight=0.113, tangentMode=RCTM_User, time=0.8, value=1}
           <at 1 -0.3 RCIM_Cubic>
          ]}}}
       Scale Curve=&RandomRangeFloat(Minimum=0.9, Maximum=1.2)
      )
     )
    )) {isSetParameters=true}
   &ScaleSpriteSize(Uniform Curve Sprite Scale=DataInterface() {propertyValues {Curve {keys [
        {arriveTangent=49.5623, interpMode=RCIM_Cubic, leaveTangent=49.5623, tangentMode=RCTM_User, time=0, value=0}
        {arriveTangentWeight=0.0236, interpMode=RCIM_Cubic, leaveTangentWeight=0.3536, tangentMode=RCTM_User, time=0.04, value=1}
        {arriveTangent=-1.0417, interpMode=RCIM_Cubic, leaveTangent=-1.0417, time=0.8, value=0.8}
        {arriveTangent=-4.4155, tangentMode=RCTM_User, time=1, value=0}
       ]}}})
   &UpdateMeshOrientation(Rotation Rate=&FloatFromCurve(
     FloatCurve=DataInterface() {propertyValues {Curve {keys [
         {arriveTangent=-6.0829, interpMode=RCIM_Cubic, time=0, value=1}
         {arriveTangent=-9.1372, interpMode=RCIM_Cubic, leaveTangent=-9.1372, tangentMode=RCTM_User, time=0.07, value=-0.15}
         {arriveTangentWeight=0.0174, interpMode=RCIM_Cubic, leaveTangentWeight=0.1566, tangentMode=RCTM_User, time=0.11, value=0}
         {arriveTangentWeight=0.2316, leaveTangentWeight=0.2059, time=0.8, value=0}
         <at 1 -1 RCIM_Cubic>
        ]}}}
     Scale Curve=&RandomRangeFloat(Minimum=0.1, Maximum=0.35)
    ))
   &ScaleColor(
    Scale Alpha=&FloatFromCurve(FloatCurve=DataInterface() {propertyValues {Curve {keys [
         <at 0.7 1 RCIM_Cubic>
         {arriveTangent=-3.7838, interpMode=RCIM_Cubic, leaveTangent=-3.7838, tangentMode=RCTM_User, time=0.85, value=0.3}
         <at 1 0>
        ]}}})
    Scale RGB=&VectorFromFloat(Value=&FloatFromCurve(FloatCurve=DataInterface() {propertyValues {Curve {keys [{arriveTangent=-13.9762, interpMode=RCIM_Cubic, leaveTangent=-13.9762, tangentMode=RCTM_User, time=0, value=5}, <at 0.5 1>]}}}))
    ScaleA=true
    ScaleRGB=true
   )
   /SetVariables(Particles.FlareColor=&RandomRangeLinearColor(Minimum=<rgba 0.3963 1.5824 5 1>, Maximum=<rgba 1.2079 4.6769 10 1>)) {isSetParameters=true, module=SetVariables_#2}
  ]
  rendererClasses [&NiagaraMeshRendererProperties, &NiagaraSpriteRendererProperties]
  renderers [
   {rendererClass=&NiagaraMeshRendererProperties, rendererIndex=0}
   {rendererClass=&NiagaraSpriteRendererProperties, rendererIndex=1}
  ]
 }
 ShardsForward {
  bEnabled=false
  emitterUpdateScript [
   &EmitterState(Life Cycle Mode=::NewEnumerator1, Loop Behavior=::NewEnumerator1, Loop Duration=0.25, UseLoopDelay=false)
   &SpawnRate(
    SpawnRate=&FloatFromCurve(
     FloatCurve=DataInterface() {propertyValues {Curve {keys [{arriveTangent=-3.6462, interpMode=RCIM_Cubic, leaveTangent=-3.6462, tangentMode=RCTM_User, time=0, value=1}, <at 1 0>]}}}
     CurveIndex=$Emitter.NormalizedLoopAge
     Scale Curve=300
    )
    Use Spawn Probability=false
   )
   &SpawnBurst_Instantaneous(Spawn Count=50, Use Spawn Probability=false, UseLoopCountLimit=false)
  ]
  particleSpawnScript [
   &InitializeParticle(
    Color Maximum=<rgba 2.1804 5.1742 8 3>
    Color Minimum=<rgba 1.0405 1.2998 4 2>
    Color Mode=::NewEnumerator2
    Lifetime Max=4
    Lifetime Min=1.5
    Lifetime Mode=::NewEnumerator1
    Mesh Scale Max=<40 40 40>
    Mesh Scale Min=<12 12 12>
    Mesh Scale Mode=::NewEnumerator2
    Sprite Size Mode=::NewEnumerator1
    Uniform Sprite Size Max=3000
    Uniform Sprite Size Min=1500
    UsePositionOffset=false
    Write Material Random=true
   )
   /SetVariables(Particles.MeshIndex=&RandomRangeInt(Minimum=0, Maximum=2)) {isSetParameters=true, module=SetVariables_#1}
   &ShapeLocation(
    Cylinder Radius=1500
    MaskSpawnGroup=false
    Non Uniform Scale=&Multiply_VectorByFloat(Vector=<1 1 0.5>, Float=$User.Scale_FX)
    Override Seed=false
    Rotation Mode=::NewEnumerator1
    Shape Primitive=::NewEnumerator1
    Surface Only=false
    UseOwnerScale=true
   )
   &AddVelocity(
    Cone Angle=60
    Module.UseConeFalloff=false
    Module.UseVelDistribution=false
    Override Seed=false
    Velocity Mode=::NewEnumerator2
    Velocity Speed=&RandomRangeFloat(Minimum=25000, Maximum=70000)
   )
   &InitialMeshOrientation(Use Rotation Vector=false)
  ]
  particleUpdateScript [
   &ParticleState()
   &ScaleSpriteSize(
    Uniform Curve Index=&Add_Float(
     A=&Multiply_Float(A=$Particles.Age, B=&RandomRangeFloat(Minimum=0.8, Maximum=1.2))
     B=&RandomRangeFloat(Minimum=-0.25, Maximum=0.25)
    )
    Uniform Curve Scale=3
    Uniform Curve Sprite Scale=DataInterface() {propertyValues {Curve {keys [
        {tangentMode=RCTM_User, time=0.2, value=0}
        <at 0.22 1 RCIM_Cubic>
        {tangentMode=RCTM_User, time=0.5, value=0}
        <at 0.68 0>
        <at 0.7 0.8 RCIM_Cubic>
        <at 0.85 0>
       ]}}}
   )
   &UpdateMeshOrientation(
    Rotation Rate=&FloatFromCurve(
     FloatCurve=DataInterface() {propertyValues {Curve {keys [
         <at 0 1 RCIM_Cubic>
         {arriveTangent=-2.1341, interpMode=RCIM_Cubic, leaveTangent=-2.1341, tangentMode=RCTM_User, time=0.35, value=0.65}
         <at 0.8 0.1 RCIM_Cubic>
        ]}}}
     Scale Curve=&RandomRangeFloat(Minimum=0.5, Maximum=1)
    )
    Rotation Vector=&RandomVector(Vector Scale=1)
   )
   &Drag(
    Drag=&FloatFromCurve(
     FloatCurve=DataInterface() {propertyValues {Curve {keys [
         {arriveTangent=4.7895, interpMode=RCIM_Cubic, leaveTangent=4.7895, tangentMode=RCTM_User, time=0, value=0.1}
         {arriveTangent=0.6, interpMode=RCIM_Cubic, leaveTangent=0.6, time=0.25, value=1}
         {arriveTangent=-4.8446, interpMode=RCIM_Cubic, leaveTangent=-4.8446, tangentMode=RCTM_User, time=0.5, value=0.4}
        ]}}}
     Scale Curve=&RandomRangeFloat(Minimum=3, Maximum=4)
    )
    Use Linear Drag=true
    Use Rotational Drag=false
   )
   &SolveForcesAndVelocity(Clamp Velocity=false, Limit Acceleration=false)
   &ScaleColor(
    Scale Alpha=&FloatFromCurve(FloatCurve=DataInterface() {propertyValues {Curve {keys [<at 0.5 1 RCIM_Cubic>,<at 1 0>]}}})
    Scale RGB=&VectorFromFloat(Value=&FloatFromCurve(FloatCurve=DataInterface() {propertyValues {Curve {keys [{arriveTangent=-13.9762, interpMode=RCIM_Cubic, time=0, value=10}, <at 0.5 1>]}}}))
    ScaleA=true
    ScaleRGB=true
   )
   /SetVariables(Particles.FlareColor=&RandomRangeLinearColor(Minimum=<rgba 0.1585 0.633 2 1>, Maximum=<rgba 0.464 3.3703 5 1>)) {isSetParameters=true}
   /SetVariables(Particles.FlareColor2:LinearColor=&MultiplyLinearColor(
     Linear Color=<rgba 0.1585 0.633 2 1>
     Scale Factor=&MakeFloat4FromVectorAndFloat(
      "Vector (RGB)"=<1 1 1>
      "Float (Alpha)"=&FloatFromCurve(FloatCurve=DataInterface() {propertyValues {Curve {keys [
           <at 0 1 RCIM_Cubic>
           {arriveTangent=-1.5311, interpMode=RCIM_Cubic, leaveTangent=-1.5311, tangentMode=RCTM_User, time=0.5, value=0.3}
           <at 1 0>
          ]}}})
     )
    )) {isSetParameters=true}
  ]
  rendererClasses [&NiagaraMeshRendererProperties, &NiagaraSpriteRendererProperties]
  renderers [
   {rendererClass=&NiagaraMeshRendererProperties, rendererIndex=0}
   {rendererClass=&NiagaraSpriteRendererProperties, rendererIndex=1}
   {rendererClass=&NiagaraSpriteRendererProperties, rendererIndex=2}
  ]
 }
 ShardsOpening {
  bEnabled=false
  emitterUpdateScript [
   &EmitterState()
   &SpawnBurst_Instantaneous(Spawn Count=25, Spawn Time=0.05, Use Spawn Probability=false, UseLoopCountLimit=false)
   &SpawnBurst_Instantaneous(Spawn Count=10, Spawn Time=0.025, Use Spawn Probability=false, UseLoopCountLimit=false) {module=SpawnBurst_Instantaneous001}
  ]
  particleSpawnScript [
   &InitializeParticle(
    Color Maximum=<rgba 3.2706 7.7612 12 2>
    Color Minimum=<rgba 1.5607 1.9496 6 2>
    Color Mode=::NewEnumerator2
    Lifetime=&Multiply_Float(A=$System.Duration, B=&RandomRangeFloat(Minimum=0.3, Maximum=0.6))
    Mesh Scale Max=<120 120 120>
    Mesh Scale Min=<32 32 32>
    Mesh Scale Mode=::NewEnumerator2
    Position Offset=<2000 0 0>
    Position Offset Coordinate Space=::Local
    Sprite Size Mode=::NewEnumerator1
    Uniform Sprite Size Max=4000
    Uniform Sprite Size Min=2500
    UsePositionOffset=true
    Write Material Random=true
   )
   /SetVariables(Particles.MeshIndex=&RandomRangeInt(Minimum=0, Maximum=&RandomRangeInt(Minimum=0, Maximum=1))) {isSetParameters=true, module=SetVariables_#1}
   &ShapeLocation(
    Cylinder Height=50
    Cylinder Radius=750
    MaskSpawnGroup=false
    Non Uniform Scale=&Multiply_VectorByFloat(Vector=<1 1 0.5>, Float=$User.Scale_FX)
    Override Seed=false
    Rotation Mode=::NewEnumerator1
    Shape Primitive=::NewEnumerator1
    Surface Only=true
    Surface Only Band Thickness=0.1
    Use Endcaps In Surface Only Mode=false
    UseOwnerScale=true
   )
   &AddVelocity(
    Module.Use Falloff Distance=false
    Origin Offset=<1000 0 0>
    Override Seed=false
    Velocity Mode=::NewEnumerator1
    Velocity Speed=&RandomRangeFloat(Minimum=2500, Maximum=15000)
   )
   &InitialMeshOrientation(
    Mesh Axis To Orient=<0 0 1>
    Mesh Orientation Mode=::NewEnumerator2
    Rotation=&RandomRangeVector(Minimum=<0 0 0>, Maximum=<0 0 1>)
    Use Rotation Vector=true
   )
   &CameraOffset(Camera Offset Amount=500)
  ]
  particleUpdateScript [
   &ParticleState()
   &ScaleMeshSize(Scale Factor=&VectorFromFloat(Value=&FloatFromCurve(FloatCurve=DataInterface() {propertyValues {Curve {keys [
          {arriveTangentWeight=0.0236, interpMode=RCIM_Cubic, leaveTangentWeight=0.3536, tangentMode=RCTM_User, time=0, value=1}
          {arriveTangent=-1, interpMode=RCIM_Cubic, leaveTangent=-1, time=0.8, value=0.8}
          {arriveTangent=-4.4155, tangentMode=RCTM_User, time=1, value=0}
         ]}}})))
   &ScaleSpriteSize(Uniform Curve Sprite Scale=DataInterface() {propertyValues {Curve {keys [
        {interpMode=RCIM_Cubic, tangentMode=RCTM_User, time=0, value=0}
        {arriveTangentWeight=0.0236, interpMode=RCIM_Cubic, leaveTangentWeight=0.3536, tangentMode=RCTM_User, time=0.05, value=1}
        {arriveTangent=-1.0526, interpMode=RCIM_Cubic, leaveTangent=-1.0526, time=0.8, value=0.8}
        {arriveTangent=-4.4155, tangentMode=RCTM_User, time=1, value=0}
       ]}}})
   &UpdateMeshOrientation(
    Rotation Rate=&FloatFromCurve(
     FloatCurve=DataInterface() {propertyValues {Curve {keys [
         <at 0 1 RCIM_Cubic>
         {arriveTangent=-2.1341, interpMode=RCIM_Cubic, leaveTangent=-2.1341, tangentMode=RCTM_User, time=0.2, value=0.7}
         <at 0.8 0.1 RCIM_Cubic>
        ]}}}
     Scale Curve=&RandomRangeFloat(Minimum=0.2, Maximum=0.7)
    )
    Rotation Vector=&RandomVector(Vector Scale=1)
   )
   &Drag(
    Drag=&FloatFromCurve(
     FloatCurve=DataInterface() {propertyValues {Curve {keys [
         {arriveTangent=7.7643, interpMode=RCIM_Cubic, leaveTangent=7.7643, tangentMode=RCTM_User, time=0, value=0.1}
         {arriveTangent=0.6, interpMode=RCIM_Cubic, leaveTangent=0.6, time=0.2, value=1}
         {arriveTangent=-4.8446, interpMode=RCIM_Cubic, leaveTangent=-4.8446, tangentMode=RCTM_User, time=0.5, value=0.4}
        ]}}}
     Scale Curve=&RandomRangeFloat(Minimum=1, Maximum=3)
    )
    Use Linear Drag=true
    Use Rotational Drag=false
   )
   &SolveForcesAndVelocity(Clamp Velocity=false, Limit Acceleration=false)
   &ScaleColor(
    Scale Alpha=&FloatFromCurve(FloatCurve=DataInterface() {propertyValues {Curve {keys [<at 0.7 1 RCIM_Cubic>,<at 1 0>]}}})
    Scale RGB=&VectorFromFloat(Value=&FloatFromCurve(FloatCurve=DataInterface() {propertyValues {Curve {keys [{arriveTangent=-13.9762, interpMode=RCIM_Cubic, leaveTangent=-13.9762, tangentMode=RCTM_User, time=0, value=5}, <at 0.5 1>]}}}))
    ScaleA=true
    ScaleRGB=true
   )
   /SetVariables(Particles.FlareColor=&RandomRangeLinearColor(Minimum=<rgba 0.0793 0.3165 1 0.4>, Maximum=<rgba 0.1392 1.0111 1.5 0.5>)) {isSetParameters=true, module=SetVariables_#2}
  ]
  rendererClasses [&NiagaraMeshRendererProperties, &NiagaraSpriteRendererProperties]
  renderers [
   {rendererClass=&NiagaraMeshRendererProperties, rendererIndex=0}
   {rendererClass=&NiagaraSpriteRendererProperties, rendererIndex=1}
  ]
 }
 ShardsRadial {
  emitterUpdateScript [
   &EmitterState(Life Cycle Mode=::NewEnumerator1, Loop Behavior=::NewEnumerator1, Loop Duration=0.5, UseLoopDelay=false)
   &SpawnRate(
    SpawnRate=&FloatFromCurve(
     FloatCurve=DataInterface() {propertyValues {Curve {keys [{arriveTangent=-3.6462, interpMode=RCIM_Cubic, leaveTangent=-3.6462, tangentMode=RCTM_User, time=0, value=1}, <at 1 0>]}}}
     CurveIndex=$Emitter.NormalizedLoopAge
     Scale Curve=300
    )
    Use Spawn Probability=false
   )
   &SpawnBurst_Instantaneous(Spawn Count=50, Use Spawn Probability=false, UseLoopCountLimit=false)
  ]
  particleSpawnScript [
   &InitializeParticle(
    Color Maximum=<rgba 2.1804 5.1742 8 1>
    Color Minimum=<rgba 1.0405 1.2998 4 1>
    Color Mode=::NewEnumerator2
    Lifetime Max=5
    Lifetime Min=3
    Lifetime Mode=::NewEnumerator1
    Mesh Scale Max=<40 40 40>
    Mesh Scale Min=<5 5 5>
    Mesh Scale Mode=::NewEnumerator2
    Sprite Size Mode=::NewEnumerator1
    Uniform Sprite Size Max=5000
    Uniform Sprite Size Min=2500
    UsePositionOffset=false
    Write Material Random=true
   )
   /SetVariables(Particles.MeshIndex=&RandomRangeInt(Minimum=0, Maximum=&RandomRangeInt(Minimum=0, Maximum=1))) {isSetParameters=true, module=SetVariables_#1}
   &ShapeLocation(
    Cylinder Radius=500
    MaskSpawnGroup=false
    Non Uniform Scale=&Multiply_VectorByFloat(Vector=<1 1 0.5>, Float=$User.Scale_FX)
    Override Seed=false
    Rotation Mode=::NewEnumerator1
    Shape Primitive=::NewEnumerator1
    Surface Only=false
    UseOwnerScale=true
   )
   &AddVelocity(
    Module.Use Falloff Distance=false
    Override Seed=false
    Velocity Mode=::NewEnumerator1
    Velocity Speed=&RandomRangeFloat(Minimum=3500, Maximum=17500)
   )
   &InitialMeshOrientation(Use Rotation Vector=false)
  ]
  particleUpdateScript [
   &ParticleState()
   &ScaleMeshSize(Scale Factor=&VectorFromFloat(Value=$User.Scale_FX))
   &ScaleSpriteSize(
    Uniform Curve Index=&Add_Float(
     A=&Multiply_Float(A=$Particles.Age, B=&RandomRangeFloat(Minimum=0.8, Maximum=1.5))
     B=&RandomRangeFloat(Minimum=-0.25, Maximum=0.25)
    )
    Uniform Curve Scale=$User.Scale_FX
    Uniform Curve Sprite Scale=DataInterface() {propertyValues {Curve {keys [
        {tangentMode=RCTM_User, time=0.2, value=0}
        <at 0.22 1 RCIM_Cubic>
        {tangentMode=RCTM_User, time=0.5, value=0}
        <at 0.68 0>
        <at 0.7 0.8 RCIM_Cubic>
        <at 0.85 0>
       ]}}}
   )
   &UpdateMeshOrientation(
    Rotation Rate=&FloatFromCurve(
     FloatCurve=DataInterface() {propertyValues {Curve {keys [
         <at 0 1 RCIM_Cubic>
         {arriveTangent=-2.1341, interpMode=RCIM_Cubic, leaveTangent=-2.1341, tangentMode=RCTM_User, time=0.35, value=0.65}
         <at 0.8 0.1 RCIM_Cubic>
        ]}}}
     Scale Curve=&RandomRangeFloat(Minimum=0.5, Maximum=1)
    )
    Rotation Vector=&RandomVector(Vector Scale=1)
   )
   &Drag(
    Drag=&FloatFromCurve(
     FloatCurve=DataInterface() {propertyValues {Curve {keys [
         {arriveTangent=4.7895, interpMode=RCIM_Cubic, leaveTangent=4.7895, tangentMode=RCTM_User, time=0, value=0.1}
         {arriveTangent=0.6, interpMode=RCIM_Cubic, leaveTangent=0.6, time=0.25, value=1}
         {arriveTangent=-4.8446, interpMode=RCIM_Cubic, leaveTangent=-4.8446, tangentMode=RCTM_User, time=0.5, value=0.4}
        ]}}}
     Scale Curve=&RandomRangeFloat(Minimum=1, Maximum=3)
    )
    Use Linear Drag=true
    Use Rotational Drag=false
   )
   &SolveForcesAndVelocity(Clamp Velocity=false, Limit Acceleration=false)
   &ScaleColor(
    Scale Alpha=&FloatFromCurve(FloatCurve=DataInterface() {propertyValues {Curve {keys [<at 0.5 1 RCIM_Cubic>,<at 1 0>]}}})
    Scale RGB=&VectorFromFloat(Value=&FloatFromCurve(FloatCurve=DataInterface() {propertyValues {Curve {keys [{arriveTangent=-13.9762, interpMode=RCIM_Cubic, leaveTangent=-13.9762, tangentMode=RCTM_User, time=0, value=5}, <at 0.5 1>]}}}))
    ScaleA=true
    ScaleRGB=true
   )
   /SetVariables(Particles.FlareColor=&RandomRangeLinearColor(Minimum=<rgba 0.3963 1.5824 5 1>, Maximum=<rgba 0.9279 6.7405 10 1>)) {isSetParameters=true, module=SetVariables_#2}
  ]
  rendererClasses [&NiagaraMeshRendererProperties, &NiagaraSpriteRendererProperties]
  renderers [
   {rendererClass=&NiagaraMeshRendererProperties, rendererIndex=0}
   {rendererClass=&NiagaraSpriteRendererProperties, rendererIndex=1}
  ]
 }
 SparklesRadial {
  emitterUpdateScript [
   &EmitterState(Life Cycle Mode=::NewEnumerator1, Loop Behavior=::NewEnumerator1, Loop Duration=0.6, UseLoopDelay=false)
   &SpawnRate(
    SpawnRate=&FloatFromCurve(
     FloatCurve=DataInterface() {propertyValues {Curve {keys [{arriveTangent=-1.1215, interpMode=RCIM_Cubic, time=0, value=1}, <at 1 0>]}}}
     CurveIndex=$Emitter.NormalizedLoopAge
     Scale Curve=800
    )
    Use Spawn Probability=false
   )
   &SpawnBurst_Instantaneous(Spawn Count=120, Use Spawn Probability=false, UseLoopCountLimit=false)
  ]
  particleSpawnScript [
   &InitializeParticle(
    Color Maximum=<rgba 0.3542 2.2297 4 1>
    Color Minimum=<rgba 0.1042 0.1826 1 1>
    Color Mode=::NewEnumerator2
    Lifetime Max=6
    Lifetime Min=3
    Lifetime Mode=::NewEnumerator1
    Sprite Size Mode=::NewEnumerator1
    Uniform Sprite Size Max=1000
    Uniform Sprite Size Min=200
    UsePositionOffset=false
    Write Material Random=true
   )
   &ShapeLocation(
    Cylinder Height=250
    Cylinder Radius=750
    MaskSpawnGroup=false
    Non Uniform Scale=&Multiply_VectorByFloat(Vector=<1 1 0.5>, Float=$User.Scale_FX)
    Override Seed=false
    Rotation Mode=::NewEnumerator1
    Shape Primitive=::NewEnumerator1
    Surface Only=false
    UseOwnerScale=true
   )
   &AddVelocity(
    Module.Use Falloff Distance=false
    Override Seed=false
    Velocity Mode=::NewEnumerator1
    Velocity Speed=&RandomRangeFloat(Minimum=7000, Maximum=35000)
   )
  ]
  particleUpdateScript [
   &ParticleState()
   &DynamicMaterialParameters(
    Index 0 Param 1=&FloatFromCurve(
     FloatCurve=DataInterface() {propertyValues {Curve {keys [{arriveTangent=-1.4828, interpMode=RCIM_Cubic, leaveTangent=-1.4828, tangentMode=RCTM_User, time=0, value=1}, <at 1 0>]}}}
     Scale Curve=&RandomRangeFloat(Minimum=0, Maximum=1)
    )
    Index 0 Param 2=&RandomRangeFloat(Minimum=0.5, Maximum=0.8)
    Module.Index0Param0Write=true
    Module.Index0Param1Write=true
    Module.Index0Param2Write=false
    Module.Index0Param3Write=false
   )
   &ScaleSpriteSize(Uniform Curve Sprite Scale=DataInterface() {propertyValues {Curve {keys [
        {interpMode=RCIM_Cubic, tangentMode=RCTM_User, time=0, value=0}
        {interpMode=RCIM_Cubic, tangentMode=RCTM_User, time=0.1, value=1}
        {arriveTangent=-1.977, tangentMode=RCTM_User, time=1, value=0}
       ]}}})
   &Drag(
    Drag=&FloatFromCurve(
     FloatCurve=DataInterface() {propertyValues {Curve {keys [
         {arriveTangent=4.7895, interpMode=RCIM_Cubic, leaveTangent=4.7895, tangentMode=RCTM_User, time=0, value=0.1}
         {arriveTangent=0.6, interpMode=RCIM_Cubic, leaveTangent=0.6, time=0.25, value=1}
         {arriveTangent=-4.8446, interpMode=RCIM_Cubic, leaveTangent=-4.8446, tangentMode=RCTM_User, time=0.5, value=0.4}
        ]}}}
     Scale Curve=&RandomRangeFloat(Minimum=1.5, Maximum=3)
    )
    Use Linear Drag=true
    Use Rotational Drag=false
   )
   &SolveForcesAndVelocity(Clamp Velocity=false, Limit Acceleration=false)
   &ScaleColor(
    Scale Alpha=&FloatFromCurve(FloatCurve=DataInterface() {propertyValues {Curve {keys [<at 0.5 1 RCIM_Cubic>,<at 1 0>]}}})
    Scale RGB=&VectorFromFloat(Value=&FloatFromCurve(FloatCurve=DataInterface() {propertyValues {Curve {keys [{arriveTangent=-13.9762, interpMode=RCIM_Cubic, leaveTangent=-13.9762, tangentMode=RCTM_User, time=0, value=5}, <at 0.5 1>]}}}))
    ScaleA=true
    ScaleRGB=true
   )
  ]
  rendererClasses [&NiagaraSpriteRendererProperties]
  renderers [{rendererClass=&NiagaraSpriteRendererProperties}]
 }
 Telegraph {
  emitterUpdateScript [
   &EmitterState()
   &SpawnBurst_Instantaneous(Use Spawn Probability=false, UseLoopCountLimit=false)
  ]
  particleSpawnScript [
   &InitializeParticle(
    Color=<rgba 5 7.6684 10 1>
    Lifetime=0.4
    Mass Mode=::NewEnumerator1
    Mesh Scale Mode=::NewEnumerator3
    Mesh Uniform Scale=&Multiply_Float(A=200, B=$User.Scale_FX)
    Position Offset=<-2000 0 0>
    UsePositionOffset=true
    Write Material Random=true
   ) {module=InitializeParticle001}
   &InitialMeshOrientation(Mesh Axis To Orient=<0 0 1>, Mesh Orientation Mode=::NewEnumerator2, Rotation=<0 0 0.4>, Use Rotation Vector=true)
  ]
  particleUpdateScript [
   &ParticleState()
   &ScaleColor(
    Scale Alpha=&FloatFromCurve(FloatCurve=DataInterface() {propertyValues {Curve {keys [<at 0.77 1 RCIM_Cubic>,<at 1 0>]}}})
    Scale RGB=&VectorFromFloat(Value=&FloatFromCurve(FloatCurve=DataInterface() {propertyValues {Curve {keys [
          {arriveTangent=-7.9561, interpMode=RCIM_Cubic, time=0.15, value=5}
          {arriveTangentWeight=0.3005, interpMode=RCIM_Cubic, leaveTangentWeight=0.3005, tangentMode=RCTM_User, time=0.5, value=1}
          {interpMode=RCIM_Cubic, leaveTangent=-3.378, time=1, value=0.2}
         ]}}}))
    ScaleA=true
    ScaleRGB=false
   )
   &DynamicMaterialParameters(
    Index 0 Param 1=&FloatFromCurve(
     FloatCurve=DataInterface() {propertyValues {Curve {keys [
         {arriveTangent=-2.1083, interpMode=RCIM_Cubic, leaveTangent=-2.1083, tangentMode=RCTM_User, time=0, value=1}
         {arriveTangent=-0.0624, tangentMode=RCTM_User, time=1, value=0}
        ]}}}
     Scale Curve=4.5
    )
    Index 0 Param 2=&FloatFromCurve(
     FloatCurve=DataInterface() {propertyValues {Curve {keys [
         {arriveTangent=12.7699, interpMode=RCIM_Cubic, leaveTangent=12.7699, tangentMode=RCTM_User, time=0, value=0}
         <at 0.17 1 RCIM_Cubic>
         {arriveTangent=-2.0178, tangentMode=RCTM_User, time=1, value=0}
        ]}}}
     Scale Curve=2
    )
    Module.Index0Param0Write=true
    Module.Index0Param1Write=true
    Module.Index0Param2Write=false
    Module.Index0Param3Write=false
   )
  ]
  rendererClasses [&NiagaraMeshRendererProperties]
  renderers [{rendererClass=&NiagaraMeshRendererProperties, rendererIndex=0}]
 }
 WaveMesh {
  emitterUpdateScript [
   &EmitterState()
   &SpawnBurst_Instantaneous(Spawn Count=5, Use Spawn Probability=false, UseLoopCountLimit=false)
  ]
  particleSpawnScript [
   &InitializeParticle(
    Color=<rgba 3.5168 20.1265 35 1.5>
    Lifetime=3.25
    Mesh Scale=&Multiply_VectorByFloat(Vector=<85 85 17>, Float=$User.Scale_FX)
    Mesh Scale Mode=::NewEnumerator4
    UsePositionOffset=false
    Write Material Random=true
   )
   &CameraOffset(Camera Offset Amount=5000)
   &InitialMeshOrientation(Mesh Axis To Orient=<0 0 1>, Mesh Orientation Mode=::NewEnumerator2, Use Rotation Vector=false)
  ]
  particleUpdateScript [
   &ParticleState()
   /SetVariables(Particles.Position=&MakeVector(
     X=&FloatFromCurve(FloatCurve=DataInterface() {propertyValues {Curve {keys [
          {arriveTangent=4.5058, interpMode=RCIM_Cubic, leaveTangent=4.5058, tangentMode=RCTM_User, time=0, value=0}
          {arriveTangent=0.9631, interpMode=RCIM_Cubic, leaveTangent=0.9631, tangentMode=RCTM_User, time=0.25, value=0.6}
          {arriveTangent=0.1617, interpMode=RCIM_Cubic, leaveTangent=0.1617, tangentMode=RCTM_User, time=1, value=1}
         ]}}})
     Y=0
     Z=0
    )) {isSetParameters=true}
   &ScaleMeshSize(Scale Factor=&VectorFromFloat(Value=&FloatFromCurve(FloatCurve=DataInterface() {propertyValues {Curve {keys [
          {arriveTangent=1.2952, interpMode=RCIM_Cubic, leaveTangent=1.2952, tangentMode=RCTM_User, time=0, value=0}
          {arriveTangent=1.4199, interpMode=RCIM_Cubic, leaveTangent=1.4199, tangentMode=RCTM_User, time=0.1968, value=0.55}
          {arriveTangent=0.2576, interpMode=RCIM_Cubic, leaveTangent=0.2576, tangentMode=RCTM_User, time=1, value=1}
         ]}}})))
   &DynamicMaterialParameters(
    Index 0 Param 1=&FloatFromCurve(
     FloatCurve=DataInterface() {propertyValues {Curve {keys [
         {arriveTangent=-2.1702, interpMode=RCIM_Cubic, time=0, value=1}
         {arriveTangent=-4.0704, interpMode=RCIM_Cubic, leaveTangent=-4.0704, tangentMode=RCTM_User, time=0.3, value=0.625}
         {arriveTangent=-2.2806, interpMode=RCIM_Cubic, leaveTangent=-2.2806, tangentMode=RCTM_User, time=1, value=-1}
        ]}}}
     Scale Curve=3
    )
    Index 0 Param 2=-0.15
    Index 0 Param 3=2
    Module.Param0WriteEnabled=true
    Module.Param1WriteEnabled=true
    Module.Param2WriteEnabled=true
    Module.Param3WriteEnabled=false
   )
   &ScaleColor(
    Scale Alpha=&FloatFromCurve(FloatCurve=DataInterface() {propertyValues {Curve {keys [
         {arriveTangent=-2.2763, interpMode=RCIM_Cubic, time=0, value=1}
         {arriveTangent=-2.3237, interpMode=RCIM_Cubic, leaveTangent=-2.3237, tangentMode=RCTM_User, time=0.2, value=0.3}
         {arriveTangent=-0.2121, leaveTangent=-0.3161, tangentMode=RCTM_User, time=0.4, value=0.05}
         {arriveTangent=-0.1021, tangentMode=RCTM_User, time=1, value=0}
        ]}}})
    Scale RGB=&VectorFromFloat(Value=&FloatFromCurve(FloatCurve=DataInterface() {propertyValues {Curve {keys [
          {arriveTangent=-2.3948, interpMode=RCIM_Cubic, time=0, value=1}
          {arriveTangent=-0.6537, interpMode=RCIM_Cubic, leaveTangent=-0.6537, tangentMode=RCTM_User, time=0.15, value=0.25}
          <at 1 0.1>
         ]}}}))
    ScaleA=true
    ScaleRGB=true
   )
  ]
  rendererClasses [&NiagaraMeshRendererProperties]
  renderers [{rendererClass=&NiagaraMeshRendererProperties, rendererIndex=0}]
 }
}
summary {
 systemName=NS_DCN_Rift_Open
 emitters [
  {emitterName=WaveMesh, rendererClasses [&NiagaraMeshRendererProperties]}
  {emitterName=ShardsRadial, rendererClasses [&NiagaraMeshRendererProperties, &NiagaraSpriteRendererProperties]}
  {emitterName=SparklesRadial, rendererClasses [&NiagaraSpriteRendererProperties]}
  {
   bEnabled=false
   emitterName=ShardsForward
   rendererClasses [&NiagaraMeshRendererProperties, &NiagaraSpriteRendererProperties]
  }
  {emitterName=Flare, rendererClasses [&NiagaraSpriteRendererProperties]}
  {emitterName=LargeCracks, rendererClasses [&NiagaraMeshRendererProperties]}
  {emitterName=Telegraph, rendererClasses [&NiagaraMeshRendererProperties]}
  {
   bEnabled=false
   emitterName=ShardsOpening
   rendererClasses [&NiagaraMeshRendererProperties, &NiagaraSpriteRendererProperties]
  }
  {
   bEnabled=false
   emitterName=ShardsFixed
   rendererClasses [&NiagaraMeshRendererProperties, &NiagaraSpriteRendererProperties]
  }
  {emitterName=CrackSpline, rendererClasses [&NiagaraMeshRendererProperties]}
  {emitterName=Flare001, rendererClasses [&NiagaraSpriteRendererProperties]}
 ]
 userVariables [User.Scale_FX {type=NiagaraFloat, defaultValue=0.3, description=""}]
}
